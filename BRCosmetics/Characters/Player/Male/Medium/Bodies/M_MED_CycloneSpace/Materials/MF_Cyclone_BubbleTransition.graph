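# Material function that transitions an input material into panning bubbles and dissolves it for the Cyclone Space character, behind the useBubbles static switch.
# Camera-vector triplanar textures and a Fresnel color blend build the bubble surface.
# BubbleDesolve, bubble mask and pre-skinned-bounds Z gradient drive the blend and opacity fade, though the gradient divides by Max minus Max.
# Both blends keep the input's vertex attributes, overriding the bubble's world-aligned normal displacement.
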MaterialFunction /BRCosmetics/Characters/Player/Male/Medium/Bodies/M_MED_CycloneSpace/Materials/MF_Cyclone_BubbleTransition
graph {nodes [
  n1 = PreSkinnedPosition()
  n2 = AppendVector(A=ScalarParameter() {parameterName=PannerX}, B=ScalarParameter() {defaultValue=0.2, parameterName=PannerY})
  Add(A=@n1, B=AppendVector(A=@n2, B=Constant()))
  Constant()
  Constant()
  Constant()
  n3 = PreSkinnedLocalBounds()
  Divide(A=@n3.Half Extents)
  n4 = FunctionInput() {inputName=Mat, inputType=FunctionInput_MaterialAttributes}
  n5 = ScalarParameter() {defaultValue=2, parameterName=BubbleColorPower}
  n6 = Constant() {r=1}
  n7 = ScalarParameter() {defaultValue=12, group=Galaxy, parameterName=BubbleTiling}
  n8 = TextureObjectParameter() {
   group=Galaxy
   parameterName=GalaxyTexture
   samplerType=SAMPLERTYPE_Masks
   texture=/FNGameplayCues/Characters/Player/Male/Medium/Bodies/M_MED_CycloneSpace/Textures/Bubbles_M
  }
  n9 = MaterialFunctionCall!(CameraVector, Time Scale=@n6, Panner Speed=@n2, Texture=@n8, Tiling=@n7, Offset, AxisFadeContrast) {materialFunction=/Game/Weapons/WeaponSkins/MaterialFunction/TriplanarCamerVector_Panning}
  n10 = Multiply(
   A=LinearInterpolate(
    A=VectorParameter() {parameterName=BubbleColor, defaultValue=<rgba 0.2129 0.375 0.3098 1>}.RGB
    B=VectorParameter() {parameterName=BubbleColor2, defaultValue=<rgba 0 0.0552 0.0885 1>}.RGB
    Alpha=Fresnel(ExponentIn=@n5) {baseReflectFraction=0}
   )
   B=@n9.Result
  )
  n11 = TextureObjectParameter() {
   group=Galaxy
   parameterName=GalaxyNormal
   samplerType=SAMPLERTYPE_Normal
   texture=/BRCosmetics/Characters/Player/Male/Medium/Bodies/M_MED_CycloneSpace/Textures/Bubbles_N
  }
  n12 = VertexNormalWS()
  n13 = TextureObjectParameter() {
   group=Galaxy
   parameterName=GalaxyTexture
   samplerType=SAMPLERTYPE_Masks
   texture=&Bubbles_M
  }
  n14 = MaterialFunctionCall!(
   TextureObject=@n13
   TextureSize=ScalarParameter() {defaultValue=500000, parameterName=WPO_SIze}
   WorldPosition=@n1
   Export Float 4
   In=@n2
   World Space Normal
   ProjectionTransitionContrast
  ) {materialFunction=/Game/Athena/SafeZone/WorldAlignedTexture_ScalePan}
  n15 = SetMaterialAttributes!(
   MaterialAttributes
   BaseColor=@n10
   Metallic=Constant()
   Specular=Constant() {r=0.4}
   Roughness=ScalarParameter() {defaultValue=0.2, parameterName=Bubble Roughness}
   Normal=MaterialFunctionCall(Time Scale=@n6, Panner Speed=@n2, Texture=@n11, Tiling=@n7) {materialFunction=&TriplanarCamerVector_Panning}.Result
   EmissiveColor=Multiply(A=@n10) {constB=0.2}
   WorldDisplacement=Multiply(A=Multiply(A=@n12, B=@n14.XYZ Texture), B=ScalarParameter() {defaultValue=50000, parameterName=WPO_intensity})
  ) {attributeSetTypes [
    #1
    #2
    #3
    #4
    #5
    #6
    #7
   ]}
  n16 = ComponentMask(None=Add(A=OneMinus(None=@n9.Result), B=ScalarParameter() {defaultValue=0.5119, parameterName=bubbleBost}) {constB=0.25}) {g=true}
  n17 = Subtract(A=VertexInterpolator(VS=PreSkinnedPosition()).PS, B=@n3.Min)
  n18 = Subtract(A=@n3.Max, B=@n3.Max)
  n19 = OneMinus(None=ComponentMask(None=Divide(A=@n17, B=@n18)) {b=true})
  n20 = Add(A=Multiply(A=ScalarParameter() {parameterName=BubbleDesolve}) {constB=2}, B=Subtract(A=Multiply(A=@n16, B=@n19)))
  n21 = BlendMaterialAttributes(
   A=SetMaterialAttributes!(MaterialAttributes=@n4, OpacityMask=Constant() {r=1}) {attributeSetTypes [#8]}
   B=@n15
   Alpha=Saturate(None=LinearInterpolate(Alpha=@n20) {constA=-0.25, constB=1.2})
  ) {vertexAttributeBlendType=UseA}
  n22 = SetMaterialAttributes(MaterialAttributes=@n21, OpacityMask=Constant()) {attributeSetTypes [#8]}
  FunctionOutput(None=StaticSwitch(
    True=BlendMaterialAttributes(A=@n21, B=@n22, Alpha=@n20) {vertexAttributeBlendType=UseA}
    False=@n4
    Value=StaticBoolParameter() {parameterName=useBubbles}
   ))
  Multiply(A=TextureCoordinate())
  Multiply(A=Time())
  Power() {constExponent=1.5}
  Reroute()
  Reroute(None=Multiply())
  ScalarParameter() {defaultValue=8, group=Galaxy, parameterName=BubbleTiling}
  TextureSample()
 ]}
