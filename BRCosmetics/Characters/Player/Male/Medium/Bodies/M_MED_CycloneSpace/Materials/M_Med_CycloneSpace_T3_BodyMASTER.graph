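# Master character material for the tier 3 body of the male medium CycloneSpace cosmetic, built on material attributes.
# PreFX output passes through a jacket function, switchable thin-film, cloth-fuzz and negative effects, then fresnel emissive.
# One FX mask texture supplies each effect's mask by channel.
# The result branches into PostFX plus Cyclone bubble transition, and separately into a switchable mobile three-light function.
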Material /BRCosmetics/Characters/Player/Male/Medium/Bodies/M_MED_CycloneSpace/Materials/M_Med_CycloneSpace_T3_BodyMASTER
graph {
 properties {
  bUseMaterialAttributes=true
  bUsedWithMeshDeformer=true
  bUsedWithMorphTargets=true
  bUsedWithSkeletalMesh=true
  pixelDepthOffsetMode=PDOM_Legacy
  usedShadingModels=Default Lit
 }
 nodes [
  n1 = MaterialFunctionCall() {materialFunction=&BaseFN_CharacterFunction_PreFX}
  n2 = TextureSampleParameter2D() {
   parameterName=Mask
   samplerType=SAMPLERTYPE_Masks
   texture=/BRCosmetics/Characters/Player/Male/Medium/Bodies/M_MED_CycloneSpace/Textures/T_Cyclone_Space_Body_FX
  }
  n3 = MaterialFunctionCall!(
   Alpha=ChannelMaskParameter(None=@n2.RGB) {maskChannel=Green, parameterName=Jacket channel, defaultValue=<rgba 0 1 0 0>}.R
   "Gmap "=@n1.Gmap
   Material=@n1.Gmap
  ) {materialFunction=/Game/Characters/Player/Male/Medium/Bodies/M_MED_CycloneUniverse/Materials/MF_Jacket}
  n4 = StaticSwitchParameter(
   True=MaterialFunctionCall!(Alpha=@n2.B, Material Attributes=@n3.Result) {materialFunction=/Game/Weapons/WeaponSkins/MaterialFunction/ThinFilmFunction}.Result
   False=@n3.Result
  ) {parameterName=Use Thin Film}
  n5 = StaticSwitchParameter(
   True=MaterialFunctionCall!(Alpha=@n2.R, Gmap=@n1.Gmap, Material=@n4) {materialFunction=/Game/Weapons/WeaponSkins/MaterialFunction/ClothFuzz}.Result
   False=@n4
  ) {parameterName=Use Cloth Fuzz}
  n6 = MaterialFunctionCall!(
   Materials=StaticSwitchParameter(
    True=MaterialFunctionCall!(Alpha=@n2.R, Gmap=@n1.Gmap, In=@n5) {materialFunction=/BRCosmetics/Weapons/WeaponSkins/MaterialFunction/MF_CyborgUniverseNegative}.Result
    False=@n5
   ) {parameterName=Use negative}
   Normal=GetMaterialAttributes(None=@n1.Result) {attributeGetTypes [#1]}.Normal
   Primary Mask=@n2.A
  ) {materialFunction=/Game/Weapons/WeaponSkins/Wraps/MaterialFunctions/MF_FresnelEmissive}
  n7 = MaterialFunctionCall(AO=@n1.AO, Distance Blend=@n1.AO, Fresnel=@n1.AO, Gmap=@n1.AO, Pre Skinned Local Position=@n1.AO, Material=@n6.Result) {materialFunction=&BaseFN_CharacterFunction_PostFX}
  MaterialFunctionCall!(Mat=@n7.Result) {materialFunction=&/MF_Cyclone_BubbleTransition}
  StaticSwitchParameter(
   True=MaterialFunctionCall!(In=@n6.Result) {materialFunction=/Game/Weapons/WeaponSkins/Wraps/MaterialFunctions/MF_3LightsMobile}.Result
   False=@n6.Result
  ) {parameterName=Use Mobile 3Lights}
 ]
}
