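# Physics asset for the male medium CycloneSpace hoodie, setting up secondary motion for its hood and upper coat bones.
# Tiny MSTR capsules act as per-bone anchors, joined to a pelvis capsule by angularly free constraints.
# Each capsule-shaped dyn bone has narrow swing and twist limits against its MSTR, except dyn_hoodie_02, which chains off dyn_hoodie_01.
# Angularly free cross-links tie each side's front hood and two coat bones together.
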PhysicsAsset /BRCosmetics/Characters/Player/Male/Medium/Bodies/M_MED_CycloneSpace/Meshes/M_MED_Cyclone_Space_Hoodie_Physics
bodies {
 dyn_coatupper_md_01_l [{shapeType=Capsule, length=5, radius=3, shapeName=dyn_coatupper_md_01_l_capsule, center=<6 0 0>, rotation=<rot 0 0 90>}]
 dyn_coatupper_md_01_l_MSTR [{shapeType=Capsule, length=0.01, radius=0.01, rotation=<rot 0 -90 0>}]
 dyn_coatupper_md_01_r [{shapeType=Capsule, length=5, radius=3, shapeName=dyn_coatupper_md_01_r_capsule, center=<6 0 0>, rotation=<rot 0 0 90>}]
 dyn_coatupper_md_01_r_MSTR [{shapeType=Capsule, length=0.01, radius=0.01, rotation=<rot 0 -90 0>}]
 dyn_coatupper_tp_01_l [{shapeType=Capsule, length=5, radius=3, shapeName=dyn_coatupper_tp_01_l_capsule, center=<6 0 0>, rotation=<rot 0 0 90>}]
 dyn_coatupper_tp_01_l_MSTR [{shapeType=Capsule, length=0.01, radius=0.01, rotation=<rot 0 -90 0>}]
 dyn_coatupper_tp_01_r [{shapeType=Capsule, length=5, radius=3, shapeName=dyn_coatupper_tp_01_r_capsule, center=<6 0 0>, rotation=<rot 0 0 90>}]
 dyn_coatupper_tp_01_r_MSTR [{shapeType=Capsule, length=0.01, radius=0.01, rotation=<rot 0 -90 0>}]
 dyn_hoodie_01 [{shapeType=Capsule, length=5, radius=3, shapeName=dyn_hoodie_01_capsule, rotation=<rot 0 0 90>}]
 dyn_hoodie_01_MSTR [{shapeType=Capsule, length=0.01, radius=0.01, rotation=<rot 0 -90 0>}]
 dyn_hoodie_01_l [{shapeType=Capsule, length=4, radius=3, shapeName=dyn_hoodie_01_l_capsule, center=<4 0 0>, rotation=<rot 0 10 95>}]
 dyn_hoodie_01_l_MSTR [{shapeType=Capsule, length=0.01, radius=0.01, rotation=<rot 0 -90 0>}]
 dyn_hoodie_01_r [{shapeType=Capsule, length=5, radius=3, shapeName=dyn_hoodie_01_r_capsule, center=<4 0 0>, rotation=<rot 0 -10 95>}]
 dyn_hoodie_01_r_MSTR [{shapeType=Capsule, length=0.01, radius=0.01, rotation=<rot 0 -90 0>}]
 dyn_hoodie_02 [{shapeType=Capsule, length=5, radius=3, shapeName=dyn_hoodie_02_capsule, rotation=<rot 0 0 90>}]
 dyn_hoodie_bk_01 [{shapeType=Capsule, length=5, radius=3, shapeName=dyn_hoodie_bk_01_capsule, center=<3 0 0>, rotation=<rot 0 0 90>}]
 dyn_hoodie_bk_01_MSTR [{shapeType=Capsule, length=0.01, radius=0.01, rotation=<rot 0 -90 0>}]
 dyn_hoodie_bk_01_l [{shapeType=Capsule, length=7, radius=3, shapeName=dyn_hoodie_bk_01_l_capsule, center=<4 0 0>, rotation=<rot 0 0 90>}]
 dyn_hoodie_bk_01_l_MSTR [{shapeType=Capsule, length=0.01, radius=0.01, rotation=<rot 0 -90 0>}]
 dyn_hoodie_bk_01_r [{shapeType=Capsule, length=7, radius=3, shapeName=dyn_hoodie_bk_01_r_capsule, center=<4 0 0>, rotation=<rot 0 0 90>}]
 dyn_hoodie_bk_01_r_MSTR [{shapeType=Capsule, length=0.01, radius=0.01, rotation=<rot 0 -90 0>}]
 dyn_hoodie_fr_01_l [{shapeType=Capsule, length=5, radius=3, shapeName=dyn_hoodie_fr_01_l_capsule, center=<2 0 0>, rotation=<rot 0 10 90>}]
 dyn_hoodie_fr_01_l_MSTR [{shapeType=Capsule, length=0.01, radius=0.01, rotation=<rot 0 -90 0>}]
 dyn_hoodie_fr_01_r [{shapeType=Capsule, length=5, radius=3, shapeName=dyn_hoodie_fr_01_r_capsule, center=<2 0 0>, rotation=<rot 0 -10 90>}]
 dyn_hoodie_fr_01_r_MSTR [{shapeType=Capsule, length=0.01, radius=0.01, rotation=<rot 0 -90 0>}]
 pelvis [{shapeType=Capsule, length=20.6777, radius=8.7444}]
}
constraints [
 {
  bone1Name=dyn_hoodie_01_l_MSTR
  bone2Name=pelvis
  swing1LimitDegrees=45
  swing1Motion=Free
  swing2LimitDegrees=45
  swing2Motion=Free
  twistLimitDegrees=45
  twistMotion=Free
 }
 {
  bone1Name=dyn_hoodie_bk_01_l_MSTR
  bone2Name=pelvis
  swing1LimitDegrees=45
  swing1Motion=Free
  swing2LimitDegrees=45
  swing2Motion=Free
  twistLimitDegrees=45
  twistMotion=Free
 }
 {
  bone1Name=dyn_hoodie_bk_01_MSTR
  bone2Name=pelvis
  swing1LimitDegrees=45
  swing1Motion=Free
  swing2LimitDegrees=45
  swing2Motion=Free
  twistLimitDegrees=45
  twistMotion=Free
 }
 {
  bone1Name=dyn_hoodie_01_r_MSTR
  bone2Name=pelvis
  swing1LimitDegrees=45
  swing1Motion=Free
  swing2LimitDegrees=45
  swing2Motion=Free
  twistLimitDegrees=45
  twistMotion=Free
 }
 {
  bone1Name=dyn_hoodie_bk_01_r_MSTR
  bone2Name=pelvis
  swing1LimitDegrees=45
  swing1Motion=Free
  swing2LimitDegrees=45
  swing2Motion=Free
  twistLimitDegrees=45
  twistMotion=Free
 }
 {
  bone1Name=dyn_hoodie_fr_01_r_MSTR
  bone2Name=pelvis
  swing1LimitDegrees=45
  swing1Motion=Free
  swing2LimitDegrees=45
  swing2Motion=Free
  twistLimitDegrees=45
  twistMotion=Free
 }
 {
  bone1Name=dyn_hoodie_fr_01_l_MSTR
  bone2Name=pelvis
  swing1LimitDegrees=45
  swing1Motion=Free
  swing2LimitDegrees=45
  swing2Motion=Free
  twistLimitDegrees=45
  twistMotion=Free
 }
 {
  bone1Name=dyn_hoodie_01_MSTR
  bone2Name=pelvis
  swing1LimitDegrees=45
  swing1Motion=Free
  swing2LimitDegrees=45
  swing2Motion=Free
  twistLimitDegrees=45
  twistMotion=Free
 }
 {bone1Name=dyn_hoodie_bk_01_l, bone2Name=dyn_hoodie_bk_01_l_MSTR, swing1LimitDegrees=5, swing2LimitDegrees=30, twistLimitDegrees=5}
 {bone1Name=dyn_hoodie_bk_01, bone2Name=dyn_hoodie_bk_01_MSTR, swing1LimitDegrees=5, swing2LimitDegrees=20, twistLimitDegrees=5}
 {bone1Name=dyn_hoodie_01_r, bone2Name=dyn_hoodie_01_r_MSTR, swing1LimitDegrees=5, swing2LimitDegrees=30, twistLimitDegrees=5}
 {bone1Name=dyn_hoodie_bk_01_r, bone2Name=dyn_hoodie_bk_01_r_MSTR, swing1LimitDegrees=5, swing2LimitDegrees=30, twistLimitDegrees=5}
 {bone1Name=dyn_hoodie_fr_01_l, bone2Name=dyn_hoodie_fr_01_l_MSTR, swing1LimitDegrees=5, swing2LimitDegrees=30, twistLimitDegrees=5}
 {bone1Name=dyn_hoodie_01, bone2Name=dyn_hoodie_01_MSTR, swing1LimitDegrees=30, swing2LimitDegrees=15, twistLimitDegrees=5}
 {bone1Name=dyn_hoodie_fr_01_r, bone2Name=dyn_hoodie_fr_01_r_MSTR, swing1LimitDegrees=5, swing2LimitDegrees=30, twistLimitDegrees=5}
 {
  bone1Name=dyn_coatupper_tp_01_l_MSTR
  bone2Name=pelvis
  swing1LimitDegrees=45
  swing1Motion=Free
  swing2LimitDegrees=45
  swing2Motion=Free
  twistLimitDegrees=45
  twistMotion=Free
 }
 {
  bone1Name=dyn_coatupper_md_01_l_MSTR
  bone2Name=pelvis
  swing1LimitDegrees=45
  swing1Motion=Free
  swing2LimitDegrees=45
  swing2Motion=Free
  twistLimitDegrees=45
  twistMotion=Free
 }
 {
  bone1Name=dyn_coatupper_tp_01_r_MSTR
  bone2Name=pelvis
  swing1LimitDegrees=45
  swing1Motion=Free
  swing2LimitDegrees=45
  swing2Motion=Free
  twistLimitDegrees=45
  twistMotion=Free
 }
 {
  bone1Name=dyn_coatupper_md_01_r_MSTR
  bone2Name=pelvis
  swing1LimitDegrees=45
  swing1Motion=Free
  swing2LimitDegrees=45
  swing2Motion=Free
  twistLimitDegrees=45
  twistMotion=Free
 }
 {bone1Name=dyn_coatupper_tp_01_l, bone2Name=dyn_coatupper_tp_01_l_MSTR, swing1LimitDegrees=5, swing2LimitDegrees=20, twistLimitDegrees=5}
 {bone1Name=dyn_coatupper_md_01_l, bone2Name=dyn_coatupper_md_01_l_MSTR, swing1LimitDegrees=5, swing2LimitDegrees=45, twistLimitDegrees=5}
 {bone1Name=dyn_coatupper_tp_01_r, bone2Name=dyn_coatupper_tp_01_r_MSTR, swing1LimitDegrees=5, swing2LimitDegrees=20, twistLimitDegrees=5}
 {bone1Name=dyn_coatupper_md_01_r, bone2Name=dyn_coatupper_md_01_r_MSTR, swing1LimitDegrees=5, swing2LimitDegrees=45, twistLimitDegrees=5}
 {
  bone1Name=dyn_hoodie_fr_01_l
  bone2Name=dyn_coatupper_tp_01_l
  swing1LimitDegrees=45
  swing1Motion=Free
  swing2LimitDegrees=45
  swing2Motion=Free
  twistLimitDegrees=45
  twistMotion=Free
 }
 {
  bone1Name=dyn_coatupper_md_01_l
  bone2Name=dyn_coatupper_tp_01_l
  swing1LimitDegrees=45
  swing1Motion=Free
  swing2LimitDegrees=45
  swing2Motion=Free
  twistLimitDegrees=45
  twistMotion=Free
 }
 {bone1Name=dyn_hoodie_02, bone2Name=dyn_hoodie_01, swing1LimitDegrees=10, swing2LimitDegrees=10, twistLimitDegrees=5}
 {
  bone1Name=dyn_coatupper_tp_01_r
  bone2Name=dyn_coatupper_md_01_r
  swing1LimitDegrees=45
  swing1Motion=Free
  swing2LimitDegrees=45
  swing2Motion=Free
  twistLimitDegrees=45
  twistMotion=Free
 }
 {
  bone1Name=dyn_hoodie_fr_01_r
  bone2Name=dyn_coatupper_tp_01_r
  swing1LimitDegrees=45
  swing1Motion=Free
  swing2LimitDegrees=45
  swing2Motion=Free
  twistLimitDegrees=45
  twistMotion=Free
 }
 {bone1Name=dyn_hoodie_01_l, bone2Name=dyn_hoodie_01_l_MSTR, swing1LimitDegrees=5, swing2LimitDegrees=30, twistLimitDegrees=5}
]
