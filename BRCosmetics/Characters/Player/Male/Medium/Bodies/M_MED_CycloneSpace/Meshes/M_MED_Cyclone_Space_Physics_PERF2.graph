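# Physics asset (PERF2 variant) for the medium male Cyclone Space body, simulating a long coat against the lower body.
# Capsules cover pelvis, legs, left foot and extra calf and pelvis colliders, plus seven coat bone chains around front, sides and back.
# Vertical links mostly limit swing and tightly restrict twist, while pelvis attachments are free.
# Free cross-links join neighboring chains row by row into a lattice, open between the two front chains.
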PhysicsAsset /BRCosmetics/Characters/Player/Male/Medium/Bodies/M_MED_CycloneSpace/Meshes/M_MED_Cyclone_Space_Physics_PERF2
bodies {
 calf_col_r [{shapeType=Capsule, length=18.5, radius=11, center=<13 2 0>, rotation=<rot 90 0 180>}]
 calf_l [{shapeType=Capsule, length=49, radius=11, center=<-25 -2 0.0003>, rotation=<rot 90 0 0>}]
 calf_r [{shapeType=Capsule, length=22.5, radius=8.5, center=<4.3 2 0>, rotation=<rot 90 0 0>}]
 dyn_coat_02_l [{shapeType=Capsule, length=26, radius=4, shapeName=dyn_coat_02_l_capsule, center=<5 0 0>, rotation=<rot 0 0 90>}]
 dyn_coat_02_r [{shapeType=Capsule, length=26, radius=4, shapeName=dyn_coat_02_r_capsule, center=<5 0 0>, rotation=<rot 0 0 90>}]
 dyn_coat_03_l [{shapeType=Capsule, length=26, radius=5, shapeName=dyn_coat_03_l_capsule, center=<5 0 0>, rotation=<rot 0 0 90>}]
 dyn_coat_03_r [{shapeType=Capsule, length=26, radius=5, shapeName=dyn_coat_03_r_capsule, center=<5 0 0>, rotation=<rot 0 0 90>}]
 dyn_coat_04_l [{shapeType=Capsule, length=26, radius=5, shapeName=dyn_coat_04_l_capsule, center=<5 0 0>, rotation=<rot 0 0 90>}]
 dyn_coat_04_r [{shapeType=Capsule, length=26, radius=5, shapeName=dyn_coat_04_r_capsule, center=<5 0 0>, rotation=<rot 0 0 90>}]
 dyn_coat_05_l [{shapeType=Capsule, length=26, radius=5, shapeName=dyn_coat_05_l_capsule, center=<11.9667 -0.0001 0>, rotation=<rot 0 0 89.9999>}]
 dyn_coat_05_r [{shapeType=Capsule, length=26, radius=5, shapeName=dyn_coat_05_r_capsule, center=<6 0 0>, rotation=<rot 0 0 90>}]
 dyn_coat_bk_02 [{shapeType=Capsule, length=26, radius=4, shapeName=dyn_coat_bk_02_capsule, center=<5 0 0>, rotation=<rot 0 0 90>}]
 dyn_coat_bk_02_l [{shapeType=Capsule, length=26, radius=4, shapeName=dyn_coat_bk_02_l_capsule, center=<5 0 0>, rotation=<rot 0 0 95>}]
 dyn_coat_bk_02_r [{shapeType=Capsule, length=26, radius=4, shapeName=dyn_coat_bk_02_r_capsule, center=<5 0 0>, rotation=<rot 0 0 65>}]
 dyn_coat_bk_03 [{shapeType=Capsule, length=26, radius=5, shapeName=dyn_coat_bk_03_capsule, center=<5 0 0>, rotation=<rot 0 0 90>}]
 dyn_coat_bk_03_l [{shapeType=Capsule, length=26, radius=5, shapeName=dyn_coat_bk_03_l_capsule, center=<5 0 0>, rotation=<rot 0 0 95>}]
 dyn_coat_bk_03_r [{shapeType=Capsule, length=26, radius=5, shapeName=dyn_coat_bk_03_r_capsule, center=<5 0 0>, rotation=<rot 0 0 65>}]
 dyn_coat_bk_04 [{shapeType=Capsule, length=26, radius=5, shapeName=dyn_coat_bk_04_capsule, center=<5 0 0>, rotation=<rot 0 0 90>}]
 dyn_coat_bk_04_l [{shapeType=Capsule, length=26, radius=5, shapeName=dyn_coat_bk_04_l_capsule, center=<5 0 0>, rotation=<rot 0 0 95>}]
 dyn_coat_bk_04_r [{shapeType=Capsule, length=26, radius=5, shapeName=dyn_coat_bk_04_r_capsule, center=<5 0 0>, rotation=<rot 0 0 65>}]
 dyn_coat_bk_05 [{shapeType=Capsule, length=26, radius=5, shapeName=dyn_coat_bk_05_capsule, center=<11.9667 -0.0001 0>, rotation=<rot 0 0 89.9999>}]
 dyn_coat_bk_05_l [{shapeType=Capsule, length=26, radius=5, shapeName=dyn_coat_bk_05_l_capsule, center=<6 0 0>, rotation=<rot 0 0 95>}]
 dyn_coat_bk_05_r [{shapeType=Capsule, length=26, radius=5, shapeName=dyn_coat_bk_05_r_capsule, center=<11.9667 -0.0001 0>, rotation=<rot 0 0 65.0003>}]
 dyn_coat_fr_01_l [{
   shapeType=Capsule
   length=6.8477
   radius=4.5963
   shapeName=dyn_coat_fr_01_l_capsule
   center=<0.9306 -1.658 -0.4819>
   rotation=<rot 0 -104.3924 -86.4448>
  }]
 dyn_coat_fr_01_r [{
   shapeType=Capsule
   length=6.8684
   radius=4.64
   shapeName=dyn_coat_fr_01_r_capsule
   center=<0.9399 1.6704 -0.4625>
   rotation=<rot 0 -75.1911 -86.4369>
  }]
 dyn_coat_fr_02_l [{shapeType=Capsule, length=15.25, radius=4.75, shapeName=dyn_coat_fr_02_l_capsule, center=<6.5 0 0>, rotation=<rot 0 -10 90>}]
 dyn_coat_fr_02_r [{shapeType=Capsule, length=15.25, radius=4.75, shapeName=dyn_coat_fr_02_r_capsule, center=<6.5 0 0>, rotation=<rot 0 10 90>}]
 dyn_coat_fr_03_l [{shapeType=Capsule, length=15.25, radius=5, shapeName=dyn_coat_fr_03_l_capsule, center=<7 0 0>, rotation=<rot 0 -10 90>}]
 dyn_coat_fr_03_r [{shapeType=Capsule, length=15.25, radius=5, shapeName=dyn_coat_fr_03_r_capsule, center=<7 0 0>, rotation=<rot 0 10 90>}]
 dyn_coat_fr_04_l [{shapeType=Capsule, length=15.25, radius=5, shapeName=dyn_coat_fr_04_l_capsule, center=<6.5 0 0>, rotation=<rot 0 -10 90>}]
 dyn_coat_fr_04_r [{shapeType=Capsule, length=15.25, radius=5, shapeName=dyn_coat_fr_04_r_capsule, center=<6.5 0 0>, rotation=<rot 0 10 90>}]
 dyn_coat_fr_05_l [{shapeType=Capsule, length=15.25, radius=5, shapeName=dyn_coat_fr_05_l_capsule, center=<8 0 0>, rotation=<rot 0 -10 90>}]
 dyn_coat_fr_05_r [{shapeType=Capsule, length=15.25, radius=5, shapeName=dyn_coat_fr_05_r_capsule, center=<8 0 0>, rotation=<rot 0 10 90>}]
 dyn_pelviscol_01 [{shapeType=Capsule, length=42.4123, radius=3.518, center=<20.4486 -0.0001 0>, rotation=<rot 90 0 180>}]
 foot_l [{shapeType=Capsule, length=19, radius=10, center=<-6 5.25 0>, rotation=<rot 0 0 90>}]
 pelvis [{shapeType=Capsule, length=11.3041, radius=8.6558, center=<-0.1224 3.5186 -0.0053>, rotation=<rot -0.0373 -90 0.2129>}]
 thigh_l [{shapeType=Capsule, length=37.6623, radius=8.268, center=<-20.4486 0.0001 0>, rotation=<rot 90 0 0>}]
 thigh_r [{shapeType=Capsule, length=36.4123, radius=9.518, center=<20.4486 -0.0001 0>, rotation=<rot 90 0 180>}]
}
constraints [
 {bone1Name=dyn_coat_03_l, bone2Name=dyn_coat_02_l, swing1LimitDegrees=45, swing2LimitDegrees=45, twistLimitDegrees=5}
 {bone1Name=dyn_coat_04_l, bone2Name=dyn_coat_03_l, swing1LimitDegrees=45, swing2LimitDegrees=45, twistLimitDegrees=5}
 {bone1Name=dyn_coat_05_l, bone2Name=dyn_coat_04_l, swing1LimitDegrees=45, swing2LimitDegrees=45, twistLimitDegrees=5}
 {bone1Name=dyn_coat_bk_03, bone2Name=dyn_coat_bk_02, swing1LimitDegrees=30, swing2LimitDegrees=30, twistLimitDegrees=5}
 {bone1Name=dyn_coat_bk_04, bone2Name=dyn_coat_bk_03, swing1LimitDegrees=30, swing2LimitDegrees=30, twistLimitDegrees=5}
 {bone1Name=dyn_coat_bk_05, bone2Name=dyn_coat_bk_04, swing1LimitDegrees=30, swing2LimitDegrees=30, twistLimitDegrees=5}
 {bone1Name=dyn_coat_bk_03_l, bone2Name=dyn_coat_bk_02_l, swing1LimitDegrees=45, swing2LimitDegrees=45, twistLimitDegrees=5}
 {bone1Name=dyn_coat_bk_04_l, bone2Name=dyn_coat_bk_03_l, swing1LimitDegrees=45, swing2LimitDegrees=45, twistLimitDegrees=5}
 {bone1Name=dyn_coat_bk_05_l, bone2Name=dyn_coat_bk_04_l, swing1LimitDegrees=45, swing2LimitDegrees=45, twistLimitDegrees=5}
 {bone1Name=dyn_coat_fr_03_l, bone2Name=dyn_coat_fr_02_l, swing1LimitDegrees=45, swing2LimitDegrees=45, twistLimitDegrees=5}
 {bone1Name=dyn_coat_fr_04_l, bone2Name=dyn_coat_fr_03_l, swing1LimitDegrees=45, swing2LimitDegrees=45, twistLimitDegrees=5}
 {bone1Name=dyn_coat_fr_05_l, bone2Name=dyn_coat_fr_04_l, swing1LimitDegrees=45, swing2LimitDegrees=45, twistLimitDegrees=5}
 {bone1Name=dyn_coat_fr_03_r, bone2Name=dyn_coat_fr_02_r, swing1LimitDegrees=45, swing2LimitDegrees=45, twistLimitDegrees=5}
 {bone1Name=dyn_coat_fr_04_r, bone2Name=dyn_coat_fr_03_r, swing1LimitDegrees=45, swing2LimitDegrees=45, twistLimitDegrees=5}
 {bone1Name=dyn_coat_fr_05_r, bone2Name=dyn_coat_fr_04_r, swing1LimitDegrees=45, swing2LimitDegrees=45, twistLimitDegrees=5}
 {bone1Name=dyn_coat_03_r, bone2Name=dyn_coat_02_r, swing1LimitDegrees=45, swing2LimitDegrees=45, twistLimitDegrees=5}
 {bone1Name=dyn_coat_04_r, bone2Name=dyn_coat_03_r, swing1LimitDegrees=45, swing2LimitDegrees=45, twistLimitDegrees=5}
 {bone1Name=dyn_coat_05_r, bone2Name=dyn_coat_04_r, swing1LimitDegrees=45, swing2LimitDegrees=45, twistLimitDegrees=5}
 {bone1Name=dyn_coat_bk_03_r, bone2Name=dyn_coat_bk_02_r, swing1LimitDegrees=30, swing2LimitDegrees=30, twistLimitDegrees=5}
 {bone1Name=dyn_coat_bk_04_r, bone2Name=dyn_coat_bk_03_r, swing1LimitDegrees=30, swing2LimitDegrees=30, twistLimitDegrees=5}
 {bone1Name=dyn_coat_bk_05_r, bone2Name=dyn_coat_bk_04_r, swing1LimitDegrees=30, swing2LimitDegrees=30, twistLimitDegrees=5}
 {
  bone1Name=calf_l
  bone2Name=thigh_l
  swing1LimitDegrees=45
  swing1Motion=Free
  swing2LimitDegrees=45
  swing2Motion=Free
  twistLimitDegrees=45
  twistMotion=Free
 }
 {
  bone1Name=dyn_coat_05_l
  bone2Name=dyn_coat_fr_05_l
  swing1LimitDegrees=45
  swing1Motion=Free
  swing2LimitDegrees=45
  swing2Motion=Free
  twistLimitDegrees=45
  twistMotion=Free
 }
 {
  bone1Name=dyn_coat_04_l
  bone2Name=dyn_coat_fr_04_l
  swing1LimitDegrees=45
  swing1Motion=Free
  swing2LimitDegrees=45
  swing2Motion=Free
  twistLimitDegrees=45
  twistMotion=Free
 }
 {
  bone1Name=dyn_coat_03_l
  bone2Name=dyn_coat_fr_03_l
  swing1LimitDegrees=45
  swing1Motion=Free
  swing2LimitDegrees=45
  swing2Motion=Free
  twistLimitDegrees=45
  twistMotion=Free
 }
 {
  bone1Name=dyn_coat_02_l
  bone2Name=dyn_coat_fr_02_l
  swing1LimitDegrees=45
  swing1Motion=Free
  swing2LimitDegrees=45
  swing2Motion=Free
  twistLimitDegrees=45
  twistMotion=Free
 }
 {
  bone1Name=dyn_coat_02_l
  bone2Name=dyn_coat_bk_02_l
  swing1LimitDegrees=45
  swing1Motion=Free
  swing2LimitDegrees=45
  swing2Motion=Free
  twistLimitDegrees=45
  twistMotion=Free
 }
 {
  bone1Name=dyn_coat_03_l
  bone2Name=dyn_coat_bk_03_l
  swing1LimitDegrees=45
  swing1Motion=Free
  swing2LimitDegrees=45
  swing2Motion=Free
  twistLimitDegrees=45
  twistMotion=Free
 }
 {
  bone1Name=dyn_coat_04_l
  bone2Name=dyn_coat_bk_04_l
  swing1LimitDegrees=45
  swing1Motion=Free
  swing2LimitDegrees=45
  swing2Motion=Free
  twistLimitDegrees=45
  twistMotion=Free
 }
 {
  bone1Name=dyn_coat_05_l
  bone2Name=dyn_coat_bk_05_l
  swing1LimitDegrees=45
  swing1Motion=Free
  swing2LimitDegrees=45
  swing2Motion=Free
  twistLimitDegrees=45
  twistMotion=Free
 }
 {
  bone1Name=dyn_coat_bk_02
  bone2Name=dyn_coat_bk_02_l
  swing1LimitDegrees=45
  swing1Motion=Free
  swing2LimitDegrees=45
  swing2Motion=Free
  twistLimitDegrees=45
  twistMotion=Free
 }
 {
  bone1Name=dyn_coat_bk_03
  bone2Name=dyn_coat_bk_03_l
  swing1LimitDegrees=45
  swing1Motion=Free
  swing2LimitDegrees=45
  swing2Motion=Free
  twistLimitDegrees=45
  twistMotion=Free
 }
 {
  bone1Name=dyn_coat_bk_04
  bone2Name=dyn_coat_bk_04_l
  swing1LimitDegrees=45
  swing1Motion=Free
  swing2LimitDegrees=45
  swing2Motion=Free
  twistLimitDegrees=45
  twistMotion=Free
 }
 {
  bone1Name=dyn_coat_bk_05
  bone2Name=dyn_coat_bk_05_l
  swing1LimitDegrees=45
  swing1Motion=Free
  swing2LimitDegrees=45
  swing2Motion=Free
  twistLimitDegrees=45
  twistMotion=Free
 }
 {
  bone1Name=dyn_coat_bk_02
  bone2Name=dyn_coat_bk_02_r
  swing1LimitDegrees=45
  swing1Motion=Free
  swing2LimitDegrees=45
  swing2Motion=Free
  twistLimitDegrees=45
  twistMotion=Free
 }
 {
  bone1Name=dyn_coat_bk_03
  bone2Name=dyn_coat_bk_03_r
  swing1LimitDegrees=45
  swing1Motion=Free
  swing2LimitDegrees=45
  swing2Motion=Free
  twistLimitDegrees=45
  twistMotion=Free
 }
 {
  bone1Name=dyn_coat_bk_04
  bone2Name=dyn_coat_bk_04_r
  swing1LimitDegrees=45
  swing1Motion=Free
  swing2LimitDegrees=45
  swing2Motion=Free
  twistLimitDegrees=45
  twistMotion=Free
 }
 {
  bone1Name=dyn_coat_bk_05
  bone2Name=dyn_coat_bk_05_r
  swing1LimitDegrees=45
  swing1Motion=Free
  swing2LimitDegrees=45
  swing2Motion=Free
  twistLimitDegrees=45
  twistMotion=Free
 }
 {
  bone1Name=dyn_coat_02_r
  bone2Name=dyn_coat_bk_02_r
  swing1LimitDegrees=45
  swing1Motion=Free
  swing2LimitDegrees=45
  swing2Motion=Free
  twistLimitDegrees=45
  twistMotion=Free
 }
 {
  bone1Name=dyn_coat_03_r
  bone2Name=dyn_coat_bk_03_r
  swing1LimitDegrees=45
  swing1Motion=Free
  swing2LimitDegrees=45
  swing2Motion=Free
  twistLimitDegrees=45
  twistMotion=Free
 }
 {
  bone1Name=dyn_coat_04_r
  bone2Name=dyn_coat_bk_04_r
  swing1LimitDegrees=45
  swing1Motion=Free
  swing2LimitDegrees=45
  swing2Motion=Free
  twistLimitDegrees=45
  twistMotion=Free
 }
 {
  bone1Name=dyn_coat_05_r
  bone2Name=dyn_coat_bk_05_r
  swing1LimitDegrees=45
  swing1Motion=Free
  swing2LimitDegrees=45
  swing2Motion=Free
  twistLimitDegrees=45
  twistMotion=Free
 }
 {
  bone1Name=dyn_coat_02_r
  bone2Name=dyn_coat_fr_02_r
  swing1LimitDegrees=45
  swing1Motion=Free
  swing2LimitDegrees=45
  swing2Motion=Free
  twistLimitDegrees=45
  twistMotion=Free
 }
 {
  bone1Name=dyn_coat_03_r
  bone2Name=dyn_coat_fr_03_r
  swing1LimitDegrees=45
  swing1Motion=Free
  swing2LimitDegrees=45
  swing2Motion=Free
  twistLimitDegrees=45
  twistMotion=Free
 }
 {
  bone1Name=dyn_coat_04_r
  bone2Name=dyn_coat_fr_04_r
  swing1LimitDegrees=45
  swing1Motion=Free
  swing2LimitDegrees=45
  swing2Motion=Free
  twistLimitDegrees=45
  twistMotion=Free
 }
 {
  bone1Name=dyn_coat_05_r
  bone2Name=dyn_coat_fr_05_r
  swing1LimitDegrees=45
  swing1Motion=Free
  swing2LimitDegrees=45
  swing2Motion=Free
  twistLimitDegrees=45
  twistMotion=Free
 }
 {
  bone1Name=foot_l
  bone2Name=calf_l
  swing1LimitDegrees=45
  swing1Motion=Free
  swing2LimitDegrees=45
  swing2Motion=Free
  twistLimitDegrees=45
  twistMotion=Free
 }
 {
  bone1Name=calf_col_r
  bone2Name=calf_r
  swing1LimitDegrees=45
  swing1Motion=Free
  swing2LimitDegrees=45
  swing2Motion=Free
  twistLimitDegrees=45
  twistMotion=Free
 }
 {
  bone1Name=thigh_l
  bone2Name=pelvis
  swing1LimitDegrees=45
  swing1Motion=Free
  swing2LimitDegrees=45
  swing2Motion=Free
  twistLimitDegrees=45
  twistMotion=Free
 }
 {
  bone1Name=thigh_r
  bone2Name=pelvis
  swing1LimitDegrees=45
  swing1Motion=Free
  swing2LimitDegrees=45
  swing2Motion=Free
  twistLimitDegrees=45
  twistMotion=Free
 }
 {
  bone1Name=dyn_pelviscol_01
  bone2Name=pelvis
  swing1LimitDegrees=45
  swing1Motion=Free
  swing2LimitDegrees=45
  swing2Motion=Free
  twistLimitDegrees=45
  twistMotion=Free
 }
 {
  bone1Name=calf_r
  bone2Name=thigh_r
  swing1LimitDegrees=45
  swing1Motion=Free
  swing2LimitDegrees=45
  swing2Motion=Free
  twistLimitDegrees=45
  twistMotion=Free
 }
 {
  bone1Name=dyn_coat_02_r
  bone2Name=pelvis
  swing1LimitDegrees=45
  swing1Motion=Free
  swing2LimitDegrees=45
  swing2Motion=Free
  twistLimitDegrees=45
  twistMotion=Free
 }
 {
  bone1Name=dyn_coat_bk_02_r
  bone2Name=pelvis
  swing1LimitDegrees=45
  swing1Motion=Free
  swing2LimitDegrees=45
  swing2Motion=Free
  twistLimitDegrees=45
  twistMotion=Free
 }
 {
  bone1Name=dyn_coat_02_l
  bone2Name=pelvis
  swing1LimitDegrees=45
  swing1Motion=Free
  swing2LimitDegrees=45
  swing2Motion=Free
  twistLimitDegrees=45
  twistMotion=Free
 }
 {
  bone1Name=dyn_coat_bk_02_l
  bone2Name=pelvis
  swing1LimitDegrees=45
  swing1Motion=Free
  swing2LimitDegrees=45
  swing2Motion=Free
  twistLimitDegrees=45
  twistMotion=Free
 }
 {
  bone1Name=dyn_coat_bk_02
  bone2Name=pelvis
  swing1LimitDegrees=45
  swing1Motion=Free
  swing2LimitDegrees=45
  swing2Motion=Free
  twistLimitDegrees=45
  twistMotion=Free
 }
 {
  bone1Name=dyn_coat_fr_02_l
  bone2Name=dyn_coat_fr_01_l
  swing1LimitDegrees=45
  swing1Motion=Free
  swing2LimitDegrees=45
  swing2Motion=Free
  twistLimitDegrees=45
  twistMotion=Free
 }
 {
  bone1Name=dyn_coat_fr_01_l
  bone2Name=pelvis
  swing1LimitDegrees=45
  swing1Motion=Free
  swing2LimitDegrees=45
  swing2Motion=Free
  twistLimitDegrees=45
  twistMotion=Free
 }
 {
  bone1Name=dyn_coat_fr_02_r
  bone2Name=dyn_coat_fr_01_r
  swing1LimitDegrees=45
  swing1Motion=Free
  swing2LimitDegrees=45
  swing2Motion=Free
  twistLimitDegrees=45
  twistMotion=Free
 }
 {
  bone1Name=dyn_coat_fr_01_r
  bone2Name=pelvis
  swing1LimitDegrees=45
  swing1Motion=Free
  swing2LimitDegrees=45
  swing2Motion=Free
  twistLimitDegrees=45
  twistMotion=Free
 }
 {
  bone1Name=dyn_coat_fr_02_l
  bone2Name=dyn_coat_fr_01_l
  swing1LimitDegrees=45
  swing1Motion=Free
  swing2LimitDegrees=45
  swing2Motion=Free
  twistLimitDegrees=45
  twistMotion=Free
 }
 {
  bone1Name=dyn_coat_fr_02_r
  bone2Name=dyn_coat_fr_01_r
  swing1LimitDegrees=45
  swing1Motion=Free
  swing2LimitDegrees=45
  swing2Motion=Free
  twistLimitDegrees=45
  twistMotion=Free
 }
]
