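# Anim Blueprint for the Cyclone Space player body, animating its coat, hoodie and necklace atop a pose copied from the parent mesh.
# Constraints, look-ats and crouch/jump offsets correct muscle and coat bones before RigidBody simulation driven by gravity and wind variables.
# Variable-driven additive rotations then layer extra motion onto coat bones.
# When `rigids_off` is set, output switches to a cheaper branch using thigh constraints and back-coat trail controllers.
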AnimBlueprint /BRCosmetics/Characters/Player/Male/Medium/Bodies/M_MED_CycloneSpace/Meshes/M_MED_Cyclone_Space_AnimBP
target_skeleton=&/M_MED_Cyclone_Space_Skeleton
graphs [
 AnimGraph {nodes [
   n1 = Get crouchingAlpha() -> (crouchingAlpha:double)
   n2 = "Constraint - dyn_coat_fr_0_l_MSTR (R)"(Alpha=@n1, ConstraintWeights_0=1) {
    alphaBoolBlend {blendOption=Linear}
    boneToModify {boneName=dyn_coat_fr_0_l_MSTR}
    constraintSetup [{transformType=Rotation, targetBone {boneName=thigh_l}}]
    constraintWeights [1]
   }
   n3 = "Constraint - dyn_coat_01_l_MSTR (R)"(ComponentPose=@n2, Alpha=@n1, ConstraintWeights_0=1) {
    alphaBoolBlend {blendOption=Linear}
    boneToModify {boneName=dyn_coat_01_l_MSTR}
    constraintSetup [{transformType=Rotation, targetBone {boneName=thigh_l}}]
    constraintWeights [1]
   }
   n4 = "Constraint - dyn_coat_01_r_MSTR (R)"(ComponentPose=@n3, Alpha=0.5, ConstraintWeights_0=1) {
    alphaBoolBlend {blendOption=Linear}
    boneToModify {boneName=dyn_coat_01_r_MSTR}
    constraintSetup [{transformType=Rotation, targetBone {boneName=thigh_r}}]
    constraintWeights [1]
   }
   "Constraint - dyn_coat_fr_0_r_MSTR (R)"(ComponentPose=@n4, Alpha=0.5, ConstraintWeights_0=1) {
    alphaBoolBlend {blendOption=Linear}
    boneToModify {boneName=dyn_coat_fr_0_r_MSTR}
    constraintSetup [{transformType=Rotation, targetBone {boneName=thigh_r}}]
    constraintWeights [1]
   }
   n5 = Get scaleCoat() -> (scaleCoat:Vector)
   n6 = "Transform (Modify) Bone - Bone: dyn_coat_bk_00"(Alpha=1, Translation="0,0,0", Rotation="0,0,0", Scale=@n5) {
    rotationMode=BMM_Additive
    rotationSpace=BCS_BoneSpace
    scaleMode=BMM_Replace
    scaleSpace=BCS_BoneSpace
    alphaBoolBlend {blendOption=Linear}
    boneToModify {boneName=dyn_coat_bk_00}
   }
   n7 = "Transform (Modify) Bone - Bone: dyn_coat_bk_01"(ComponentPose=@n6, Alpha=1, Translation="0,0,0", Rotation="0,0,0", Scale=@n5) {
    rotationMode=BMM_Additive
    rotationSpace=BCS_BoneSpace
    scaleMode=BMM_Replace
    scaleSpace=BCS_BoneSpace
    alphaBoolBlend {blendOption=Linear}
    boneToModify {boneName=dyn_coat_bk_01}
   }
   n8 = "Transform (Modify) Bone - Bone: dyn_coat_bk_02"(ComponentPose=@n7, Alpha=1, Translation="0,0,0", Rotation="0,0,0", Scale=@n5) {
    rotationMode=BMM_Additive
    rotationSpace=BCS_BoneSpace
    scaleMode=BMM_Replace
    scaleSpace=BCS_BoneSpace
    alphaBoolBlend {blendOption=Linear}
    boneToModify {boneName=dyn_coat_bk_02}
   }
   n9 = "Transform (Modify) Bone - Bone: dyn_coat_bk_03"(ComponentPose=@n8, Alpha=1, Translation="0,0,0", Rotation="0,0,0", Scale=@n5) {
    rotationMode=BMM_Additive
    rotationSpace=BCS_BoneSpace
    scaleMode=BMM_Replace
    scaleSpace=BCS_BoneSpace
    alphaBoolBlend {blendOption=Linear}
    boneToModify {boneName=dyn_coat_bk_03}
   }
   n10 = "Transform (Modify) Bone - Bone: dyn_coat_bk_04"(ComponentPose=@n9, Alpha=1, Translation="0,0,0", Rotation="0,0,0", Scale=@n5) {
    rotationMode=BMM_Additive
    rotationSpace=BCS_BoneSpace
    scaleMode=BMM_Replace
    scaleSpace=BCS_BoneSpace
    alphaBoolBlend {blendOption=Linear}
    boneToModify {boneName=dyn_coat_bk_04}
   }
   n11 = "Transform (Modify) Bone - Bone: dyn_coat_bk_05"(ComponentPose=@n10, Alpha=1, Translation="0,0,0", Rotation="0,0,0", Scale=@n5) {
    rotationMode=BMM_Additive
    rotationSpace=BCS_BoneSpace
    scaleMode=BMM_Replace
    scaleSpace=BCS_BoneSpace
    alphaBoolBlend {blendOption=Linear}
    boneToModify {boneName=dyn_coat_bk_05}
   }
   n12 = "Transform (Modify) Bone - Bone: dyn_coat_bk_00_l"(ComponentPose=@n11, Alpha=1, Translation="0,0,0", Rotation="0,0,0", Scale=@n5) {rotationMode=BMM_Additive, rotationSpace=BCS_BoneSpace, alphaBoolBlend {blendOption=Linear}, boneToModify {boneName=dyn_coat_bk_00_l}}
   n13 = "Transform (Modify) Bone - Bone: dyn_coat_bk_01_l"(ComponentPose=@n12, Alpha=1, Translation="0,0,0", Rotation="0,0,0", Scale=@n5) {rotationMode=BMM_Additive, rotationSpace=BCS_BoneSpace, alphaBoolBlend {blendOption=Linear}, boneToModify {boneName=dyn_coat_bk_01_l}}
   n14 = "Transform (Modify) Bone - Bone: dyn_coat_bk_02_l"(ComponentPose=@n13, Alpha=1, Translation="0,0,0", Rotation="0,0,0", Scale=@n5) {rotationMode=BMM_Additive, rotationSpace=BCS_BoneSpace, alphaBoolBlend {blendOption=Linear}, boneToModify {boneName=dyn_coat_bk_02_l}}
   n15 = "Transform (Modify) Bone - Bone: dyn_coat_bk_03_l"(ComponentPose=@n14, Alpha=1, Translation="0,0,0", Rotation="0,0,0", Scale=@n5) {rotationMode=BMM_Additive, rotationSpace=BCS_BoneSpace, alphaBoolBlend {blendOption=Linear}, boneToModify {boneName=dyn_coat_bk_03_l}}
   n16 = "Transform (Modify) Bone - Bone: dyn_coat_bk_04_l"(ComponentPose=@n15, Alpha=1, Translation="0,0,0", Rotation="0,0,0", Scale=@n5) {rotationMode=BMM_Additive, rotationSpace=BCS_BoneSpace, alphaBoolBlend {blendOption=Linear}, boneToModify {boneName=dyn_coat_bk_04_l}}
   n17 = "Transform (Modify) Bone - Bone: dyn_coat_bk_05_l"(ComponentPose=@n16, Alpha=1, Translation="0,0,0", Rotation="0,0,0", Scale=@n5) {rotationMode=BMM_Additive, rotationSpace=BCS_BoneSpace, alphaBoolBlend {blendOption=Linear}, boneToModify {boneName=dyn_coat_bk_05_l}}
   n18 = "Transform (Modify) Bone - Bone: dyn_coat_00_l"(ComponentPose=@n17, Alpha=1, Translation="0,0,0", Rotation="0,0,0", Scale="1,1,1") {rotationMode=BMM_Additive, rotationSpace=BCS_BoneSpace, alphaBoolBlend {blendOption=Linear}, boneToModify {boneName=dyn_coat_00_l}}
   n19 = "Transform (Modify) Bone - Bone: dyn_coat_01_l"(ComponentPose=@n18, Alpha=1, Translation="0,0,0", Rotation="0,0,0", Scale="1,1,1") {rotationMode=BMM_Additive, rotationSpace=BCS_BoneSpace, alphaBoolBlend {blendOption=Linear}, boneToModify {boneName=dyn_coat_01_l}}
   n20 = "Transform (Modify) Bone - Bone: dyn_coat_02_l"(ComponentPose=@n19, Alpha=1, Translation="0,0,0", Rotation="0,0,0", Scale="1,1,1") {rotationMode=BMM_Additive, rotationSpace=BCS_BoneSpace, alphaBoolBlend {blendOption=Linear}, boneToModify {boneName=dyn_coat_02_l}}
   n21 = "Transform (Modify) Bone - Bone: dyn_coat_03_l"(ComponentPose=@n20, Alpha=1, Translation="0,0,0", Rotation="0,0,0", Scale="1,1,1") {rotationMode=BMM_Additive, rotationSpace=BCS_BoneSpace, alphaBoolBlend {blendOption=Linear}, boneToModify {boneName=dyn_coat_03_l}}
   n22 = "Transform (Modify) Bone - Bone: dyn_coat_04_l"(ComponentPose=@n21, Alpha=1, Translation="0,0,0", Rotation="0,0,0", Scale="1,1,1") {rotationMode=BMM_Additive, rotationSpace=BCS_BoneSpace, alphaBoolBlend {blendOption=Linear}, boneToModify {boneName=dyn_coat_04_l}}
   n23 = "Transform (Modify) Bone - Bone: dyn_coat_05_l"(ComponentPose=@n22, Alpha=1, Translation="0,0,0", Rotation="0,0,0", Scale="1,1,1") {rotationMode=BMM_Additive, rotationSpace=BCS_BoneSpace, alphaBoolBlend {blendOption=Linear}, boneToModify {boneName=dyn_coat_05_l}}
   n24 = "Transform (Modify) Bone - Bone: dyn_coat_fr_00"(ComponentPose=@n23, Alpha=1, Translation="0,0,0", Rotation="0,0,0", Scale="1,1,1") {rotationMode=BMM_Additive, rotationSpace=BCS_BoneSpace, alphaBoolBlend {blendOption=Linear}, boneToModify {boneName=dyn_coat_fr_00}}
   n25 = "Transform (Modify) Bone - Bone: dyn_coat_fr_01_l"(ComponentPose=@n24, Alpha=1, Translation="0,0,0", Rotation="0,0,0", Scale="1,1,1") {rotationMode=BMM_Additive, rotationSpace=BCS_BoneSpace, alphaBoolBlend {blendOption=Linear}, boneToModify {boneName=dyn_coat_fr_01_l}}
   n26 = "Transform (Modify) Bone - Bone: dyn_coat_fr_02_l"(ComponentPose=@n25, Alpha=1, Translation="0,0,0", Rotation="0,0,0", Scale="1,1,1") {rotationMode=BMM_Additive, rotationSpace=BCS_BoneSpace, alphaBoolBlend {blendOption=Linear}, boneToModify {boneName=dyn_coat_fr_02_l}}
   n27 = "Transform (Modify) Bone - Bone: dyn_coat_fr_03_l"(ComponentPose=@n26, Alpha=1, Translation="0,0,0", Rotation="0,0,0", Scale="1,1,1") {rotationMode=BMM_Additive, rotationSpace=BCS_BoneSpace, alphaBoolBlend {blendOption=Linear}, boneToModify {boneName=dyn_coat_fr_03_l}}
   n28 = "Transform (Modify) Bone - Bone: dyn_coat_fr_04_l"(ComponentPose=@n27, Alpha=1, Translation="0,0,0", Rotation="0,0,0", Scale="1,1,1") {rotationMode=BMM_Additive, rotationSpace=BCS_BoneSpace, alphaBoolBlend {blendOption=Linear}, boneToModify {boneName=dyn_coat_fr_04_l}}
   n29 = "Transform (Modify) Bone - Bone: dyn_coat_fr_05_l"(ComponentPose=@n28, Alpha=1, Translation="0,0,0", Rotation="0,0,0", Scale="1,1,1") {rotationMode=BMM_Additive, rotationSpace=BCS_BoneSpace, alphaBoolBlend {blendOption=Linear}, boneToModify {boneName=dyn_coat_fr_05_l}}
   n30 = "Transform (Modify) Bone - Bone: dyn_coat_bk_00_r"(ComponentPose=@n29, Alpha=1, Translation="0,0,0", Rotation="0,0,0", Scale="1,1,1") {rotationMode=BMM_Additive, rotationSpace=BCS_BoneSpace, alphaBoolBlend {blendOption=Linear}, boneToModify {boneName=dyn_coat_bk_00_r}}
   n31 = "Transform (Modify) Bone - Bone: dyn_coat_01_r"(ComponentPose=@n30, Alpha=1, Translation="0,0,0", Rotation="0,0,0", Scale="1,1,1") {rotationMode=BMM_Additive, rotationSpace=BCS_BoneSpace, alphaBoolBlend {blendOption=Linear}, boneToModify {boneName=dyn_coat_01_r}}
   n32 = "Transform (Modify) Bone - Bone: dyn_coat_bk_02_r"(ComponentPose=@n31, Alpha=1, Translation="0,0,0", Rotation="0,0,0", Scale="1,1,1") {rotationMode=BMM_Additive, rotationSpace=BCS_BoneSpace, alphaBoolBlend {blendOption=Linear}, boneToModify {boneName=dyn_coat_bk_02_r}}
   n33 = "Transform (Modify) Bone - Bone: dyn_coat_bk_03_r"(ComponentPose=@n32, Alpha=1, Translation="0,0,0", Rotation="0,0,0", Scale="1,1,1") {rotationMode=BMM_Additive, rotationSpace=BCS_BoneSpace, alphaBoolBlend {blendOption=Linear}, boneToModify {boneName=dyn_coat_bk_03_r}}
   n34 = "Transform (Modify) Bone - Bone: dyn_coat_bk_04_r"(ComponentPose=@n33, Alpha=1, Translation="0,0,0", Rotation="0,0,0", Scale="1,1,1") {rotationMode=BMM_Additive, rotationSpace=BCS_BoneSpace, alphaBoolBlend {blendOption=Linear}, boneToModify {boneName=dyn_coat_bk_04_r}}
   n35 = "Transform (Modify) Bone - Bone: dyn_coat_bk_05_r"(ComponentPose=@n34, Alpha=1, Translation="0,0,0", Rotation="0,0,0", Scale="1,1,1") {rotationMode=BMM_Additive, rotationSpace=BCS_BoneSpace, alphaBoolBlend {blendOption=Linear}, boneToModify {boneName=dyn_coat_bk_05_r}}
   n36 = "Transform (Modify) Bone - Bone: dyn_coat_00_r"(ComponentPose=@n35, Alpha=1, Translation="0,0,0", Rotation="0,0,0", Scale="1,1,1") {rotationMode=BMM_Additive, rotationSpace=BCS_BoneSpace, alphaBoolBlend {blendOption=Linear}, boneToModify {boneName=dyn_coat_00_r}}
   n37 = "Transform (Modify) Bone - Bone: dyn_coat_01_r"(ComponentPose=@n36, Alpha=1, Translation="0,0,0", Rotation="0,0,0", Scale="1,1,1") {rotationMode=BMM_Additive, rotationSpace=BCS_BoneSpace, alphaBoolBlend {blendOption=Linear}, boneToModify {boneName=dyn_coat_01_r}}
   n38 = "Transform (Modify) Bone - Bone: dyn_coat_02_r"(ComponentPose=@n37, Alpha=1, Translation="0,0,0", Rotation="0,0,0", Scale="1,1,1") {rotationMode=BMM_Additive, rotationSpace=BCS_BoneSpace, alphaBoolBlend {blendOption=Linear}, boneToModify {boneName=dyn_coat_02_r}}
   n39 = "Transform (Modify) Bone - Bone: dyn_coat_03_r"(ComponentPose=@n38, Alpha=1, Translation="0,0,0", Rotation="0,0,0", Scale="1,1,1") {rotationMode=BMM_Additive, rotationSpace=BCS_BoneSpace, alphaBoolBlend {blendOption=Linear}, boneToModify {boneName=dyn_coat_03_r}}
   n40 = "Transform (Modify) Bone - Bone: dyn_coat_04_r"(ComponentPose=@n39, Alpha=1, Translation="0,0,0", Rotation="0,0,0", Scale="1,1,1") {rotationMode=BMM_Additive, rotationSpace=BCS_BoneSpace, alphaBoolBlend {blendOption=Linear}, boneToModify {boneName=dyn_coat_04_r}}
   n41 = "Transform (Modify) Bone - Bone: dyn_coat_05_r"(ComponentPose=@n40, Alpha=1, Translation="0,0,0", Rotation="0,0,0", Scale="1,1,1") {rotationMode=BMM_Additive, rotationSpace=BCS_BoneSpace, alphaBoolBlend {blendOption=Linear}, boneToModify {boneName=dyn_coat_05_r}}
   n42 = "Transform (Modify) Bone - Bone: dyn_coat_fr_00_r"(ComponentPose=@n41, Alpha=1, Translation="0,0,0", Rotation="0,0,0", Scale="1,1,1") {rotationMode=BMM_Additive, rotationSpace=BCS_BoneSpace, alphaBoolBlend {blendOption=Linear}, boneToModify {boneName=dyn_coat_fr_00_r}}
   n43 = "Transform (Modify) Bone - Bone: dyn_coat_fr_01_r"(ComponentPose=@n42, Alpha=1, Translation="0,0,0", Rotation="0,0,0", Scale="1,1,1") {rotationMode=BMM_Additive, rotationSpace=BCS_BoneSpace, alphaBoolBlend {blendOption=Linear}, boneToModify {boneName=dyn_coat_fr_01_r}}
   n44 = "Transform (Modify) Bone - Bone: dyn_coat_fr_02_r"(ComponentPose=@n43, Alpha=1, Translation="0,0,0", Rotation="0,0,0", Scale="1,1,1") {rotationMode=BMM_Additive, rotationSpace=BCS_BoneSpace, alphaBoolBlend {blendOption=Linear}, boneToModify {boneName=dyn_coat_fr_02_r}}
   n45 = "Transform (Modify) Bone - Bone: dyn_coat_fr_03_r"(ComponentPose=@n44, Alpha=1, Translation="0,0,0", Rotation="0,0,0", Scale="1,1,1") {rotationMode=BMM_Additive, rotationSpace=BCS_BoneSpace, alphaBoolBlend {blendOption=Linear}, boneToModify {boneName=dyn_coat_fr_03_r}}
   n46 = "Transform (Modify) Bone - Bone: dyn_coat_fr_04_r"(ComponentPose=@n45, Alpha=1, Translation="0,0,0", Rotation="0,0,0", Scale="1,1,1") {rotationMode=BMM_Additive, rotationSpace=BCS_BoneSpace, alphaBoolBlend {blendOption=Linear}, boneToModify {boneName=dyn_coat_fr_04_r}}
   "Transform (Modify) Bone - Bone: dyn_coat_fr_05_r"(ComponentPose=@n46, Alpha=1, Translation="0,0,0", Rotation="0,0,0", Scale="1,1,1") {rotationMode=BMM_Additive, rotationSpace=BCS_BoneSpace, alphaBoolBlend {blendOption=Linear}, boneToModify {boneName=dyn_coat_fr_05_r}}
   n47 = Get coatUp() -> (coatUp:Vector)
   n48 = Get noiseRotRoot() -> (noiseRotRoot:Rotator)
   n49 = "Transform (Modify) Bone - Bone: dyn_coat_bk_01_MSTR"(Alpha=1, Translation=@n47, Rotation=@n48, Scale="1,1,1") {
    rotationMode=BMM_Additive
    rotationSpace=BCS_BoneSpace
    scaleMode=BMM_Replace
    scaleSpace=BCS_BoneSpace
    translationMode=BMM_Additive
    translationSpace=BCS_BoneSpace
    alphaBoolBlend {blendOption=Linear}
    boneToModify {boneName=dyn_coat_bk_01_MSTR}
   }
   n50 = "Transform (Modify) Bone - Bone: dyn_coat_bk_01_l_MSTR"(ComponentPose=@n49, Alpha=1, Translation=@n47, Rotation=@n48, Scale="1,1,1") {
    rotationMode=BMM_Additive
    rotationSpace=BCS_BoneSpace
    translationMode=BMM_Additive
    translationSpace=BCS_BoneSpace
    alphaBoolBlend {blendOption=Linear}
    boneToModify {boneName=dyn_coat_bk_01_l_MSTR}
   }
   n51 = "Transform (Modify) Bone - Bone: dyn_coat_01_l_MSTR"(ComponentPose=@n50, Alpha=1, Translation=@n47, Rotation="0,0,0", Scale="1,1,1") {
    rotationMode=BMM_Additive
    rotationSpace=BCS_BoneSpace
    translationMode=BMM_Additive
    translationSpace=BCS_BoneSpace
    alphaBoolBlend {blendOption=Linear}
    boneToModify {boneName=dyn_coat_01_l_MSTR}
   }
   n52 = "Transform (Modify) Bone - Bone: dyn_coat_fr_0_l_MSTR"(ComponentPose=@n51, Alpha=1, Translation=@n47, Rotation="0,0,0", Scale="1,1,1") {
    rotationMode=BMM_Additive
    rotationSpace=BCS_BoneSpace
    translationMode=BMM_Additive
    translationSpace=BCS_BoneSpace
    alphaBoolBlend {blendOption=Linear}
    boneToModify {boneName=dyn_coat_fr_0_l_MSTR}
   }
   n53 = "Transform (Modify) Bone - Bone: dyn_coat_bk_01_r_MSTR"(ComponentPose=@n52, Alpha=1, Translation=@n47, Rotation=@n48, Scale="1,1,1") {
    rotationMode=BMM_Additive
    rotationSpace=BCS_BoneSpace
    translationMode=BMM_Additive
    translationSpace=BCS_BoneSpace
    alphaBoolBlend {blendOption=Linear}
    boneToModify {boneName=dyn_coat_bk_01_r_MSTR}
   }
   n54 = "Transform (Modify) Bone - Bone: dyn_coat_01_r_MSTR"(ComponentPose=@n53, Alpha=1, Translation=@n47, Rotation="0,0,0", Scale="1,1,1") {
    rotationMode=BMM_Additive
    rotationSpace=BCS_BoneSpace
    translationMode=BMM_Additive
    translationSpace=BCS_BoneSpace
    alphaBoolBlend {blendOption=Linear}
    boneToModify {boneName=dyn_coat_01_r_MSTR}
   }
   "Transform (Modify) Bone - Bone: dyn_coat_fr_00_r"(ComponentPose=@n54, Alpha=1, Translation=@n47, Rotation="0,0,0", Scale="1,1,1") {
    rotationMode=BMM_Additive
    rotationSpace=BCS_BoneSpace
    translationMode=BMM_Additive
    translationSpace=BCS_BoneSpace
    alphaBoolBlend {blendOption=Linear}
    boneToModify {boneName=dyn_coat_fr_00_r}
   }
   n55 = Get windVector() -> (windVector:Vector)
   RigidBody!(ComponentPose, Alpha, OverrideWorldGravity:Vector=@n55, ExternalForce) -> (Pose) {
    bClampLinearTranslationLimitToRefPose=true
    bOverrideWorldGravity=true
    overridePhysicsAsset=&/M_MED_Cyclone_Space_Hoodie_Physics
    simulationSpace=BaseBoneSpace
    alphaBoolBlend {blendOption=Linear}
    baseBoneRef {boneName=spine_04}
    componentAppliedLinearAccClamp=<3000 3000 3000>
    componentLinearAccScale=<0.3 0.3 0.3>
    componentLinearVelScale=<0.45 0.45 0.45>
    overrideWorldGravity {z=-980}
   }
   Output Pose(Result=Blend Poses by bool(
     BlendPose_0=Use cached pose 'trailControlCache'()
     BlendPose_1=Use cached pose 'rbanCache'()
     bActiveValue=Get rigids_off() -> (rigids_off:bool)
    ) {blendType=Linear}) {layerGroup=None}
   n56 = "Constraint - scapula_base_l (P)"(ComponentPose=Local To Component(LocalPose=Use cached pose 'initCache'()), Alpha=1, ConstraintWeights_0=1) {
    alphaBoolBlend {blendOption=Linear}
    boneToModify {boneName=scapula_base_l}
    constraintSetup [{targetBone {boneName=clavicle_l}}]
    constraintWeights [1]
   }
   n57 = Look At - Bone: scapula_00_l(ComponentPose=@n56, Alpha=1) {
    alphaBoolBlend {blendOption=Linear}
    boneToModify {boneName=scapula_00_l}
    lookAtLocation {x=0}
    lookAtTarget {boneReference {boneName=scapula_01_l_aim}}
    lookAt_Axis {axis=<1 0>}
    lookUp_Axis {axis {x=0, z=1}}
   }
   n58 = Look At - Bone: pectoralis_01_l(ComponentPose=@n57, Alpha=1) {
    alphaBoolBlend {blendOption=Linear}
    boneToModify {boneName=pectoralis_01_l}
    lookAtLocation {x=0}
    lookAtTarget {boneReference {boneName=upperarm_l}}
    lookAt_Axis {axis=<1 0>}
    lookUp_Axis {axis {x=0, z=1}}
   }
   n59 = Look At - Bone: lattissimus_00_l(ComponentPose=@n58, Alpha=1) {
    alphaBoolBlend {blendOption=Linear}
    boneToModify {boneName=lattissimus_00_l}
    lookAtLocation {x=0}
    lookAtTarget {boneReference {boneName=upperarm_l}}
    lookAt_Axis {axis=<1 0>}
    lookUp_Axis {axis {x=0, z=1}}
   }
   n60 = "Constraint - scapula_base_r (P)"(ComponentPose=@n59, Alpha=1, ConstraintWeights_0=1) {
    alphaBoolBlend {blendOption=Linear}
    boneToModify {boneName=scapula_base_r}
    constraintSetup [{targetBone {boneName=clavicle_r}}]
    constraintWeights [1]
   }
   n61 = Look At - Bone: scapula_00_r(ComponentPose=@n60, Alpha=1) {
    alphaBoolBlend {blendOption=Linear}
    boneToModify {boneName=scapula_00_r}
    lookAtLocation {x=0}
    lookAtTarget {boneReference {boneName=scapula_01_r_aim}}
    lookAt_Axis {axis=<1 0>}
    lookUp_Axis {axis {x=0, z=1}}
   }
   n62 = Look At - Bone: pectoralis_01_r(ComponentPose=@n61, Alpha=1) {
    alphaBoolBlend {blendOption=Linear}
    boneToModify {boneName=pectoralis_01_r}
    lookAtLocation {x=0}
    lookAtTarget {boneReference {boneName=upperarm_r}}
    lookAt_Axis {axis=<1 0>}
    lookUp_Axis {axis {x=0, z=1}}
   }
   n63 = Look At - Bone: lattissimus_00_r(ComponentPose=@n62, Alpha=1) {
    alphaBoolBlend {blendOption=Linear}
    boneToModify {boneName=lattissimus_00_r}
    lookAtLocation {x=0}
    lookAtTarget {boneReference {boneName=upperarm_r}}
    lookAt_Axis {axis=<1 0>}
    lookUp_Axis {axis {x=0, z=1}}
   }
   n64 = "Constraint - dyn_pelviscol_01_aim (PP)"!(ComponentPose=@n63, Alpha=1, ConstraintWeights_0=0.5, ConstraintWeights_1:float=0.5) -> (Pose) {
    alphaBoolBlend {blendOption=Linear}
    boneToModify {boneName=dyn_pelviscol_01_aim}
    constraintSetup [{targetBone {boneName=calf_l}}, {targetBone {boneName=calf_r}}]
    constraintWeights [1,1]
   }
   n65 = Look At - Bone: dyn_pelviscol_01(ComponentPose=@n64, Alpha=1) {
    alphaBoolBlend {blendOption=Linear}
    boneToModify {boneName=dyn_pelviscol_01}
    lookAtTarget {boneReference {boneName=dyn_pelviscol_01_aim}}
    lookAt_Axis {axis=<1 0>}
    lookUp_Axis {axis {x=0, z=1}}
   }
   n66 = Get bIsCrouching()
   n67 = "Transform (Modify) Bone - Bone: dyn_coat_fr_01_r"!(ComponentPose=@n65, bAlphaBoolEnabled:bool=@n66, Translation="1.532,5.123,0.826", Rotation="9.847,10.151,1.754", Scale="1,1,1") -> (Pose) {
    alphaInputType=Bool
    rotationMode=BMM_Additive
    rotationSpace=BCS_BoneSpace
    translationMode=BMM_Additive
    translationSpace=BCS_BoneSpace
    alphaBoolBlend {blendInTime=0.5, blendOption=Linear, blendOutTime=0.5}
    boneToModify {boneName=dyn_coat_fr_01_r}
    rotation=<rot 9.847 10.1513 1.754>
    translation=<1.532 5.123 0.826>
   }
   n68 = "Transform (Modify) Bone - Bone: dyn_coat_fr_01_l"!(ComponentPose=@n67, bAlphaBoolEnabled:bool=@n66, Translation="-3.347,-11.185,4.163", Rotation="0,0,0", Scale="1,1,1") -> (Pose) {
    alphaInputType=Bool
    rotationMode=BMM_Additive
    rotationSpace=BCS_BoneSpace
    translationMode=BMM_Additive
    translationSpace=BCS_BoneSpace
    alphaBoolBlend {blendInTime=0.5, blendOption=Linear, blendOutTime=0.5}
    boneToModify {boneName=dyn_coat_fr_01_l}
    translation=<-3.347 -11.185 4.163>
   }
   n69 = "Transform (Modify) Bone - Bone: dyn_coat_fr_01_r"!(
    ComponentPose=@n68
    Alpha:float=Get jump_alpha() -> (jump_alpha:double)
    Translation="-1.963,9.176,0.211"
    Rotation="9.847,10.151,1.754"
    Scale="1,1,1"
   ) -> (Pose) {
    rotationMode=BMM_Additive
    rotationSpace=BCS_BoneSpace
    translationMode=BMM_Additive
    translationSpace=BCS_BoneSpace
    alphaBoolBlend {blendInTime=0.25, blendOption=Linear, blendOutTime=0.25}
    alphaScaleBiasClamp {bInterpResult=true}
    boneToModify {boneName=dyn_coat_fr_01_r}
    rotation=<rot 9.847 10.1513 1.754>
    translation=<-1.963 9.176 0.211>
   }
   n70 = "Transform (Modify) Bone - Bone: dyn_coat_fr_01_r"(
    ComponentPose=@n69
    Alpha=Get thigh_r_crouch_pitch_alpha() -> (thigh_r_crouch_pitch_alpha:double)
    Translation="0.954,9.166,5.031"
    Rotation="9.847,10.153,1.754"
    Scale="1,1,1"
   ) {
    rotationMode=BMM_Additive
    rotationSpace=BCS_BoneSpace
    translationMode=BMM_Additive
    translationSpace=BCS_BoneSpace
    alphaBoolBlend {blendInTime=0.25, blendOption=Linear, blendOutTime=0.25}
    alphaScaleBiasClamp {bInterpResult=true}
    boneToModify {boneName=dyn_coat_fr_01_r}
    rotation=<rot 9.847 10.1531 1.754>
    translation=<0.954 9.166 5.031>
   }
   Save cached pose 'presimCache'(Pose=Component To Local(ComponentPose=@n70)) {cacheName=presimCache}
   n71 = "Constraint - dyn_coat_01_l (P)"(ComponentPose=Local To Component(LocalPose=Use cached pose 'initCache'()), Alpha=1, ConstraintWeights_0=1) {
    alphaBoolBlend {blendOption=Linear}
    boneToModify {boneName=dyn_coat_01_l}
    constraintSetup [{targetBone {boneName=thigh_l}}]
    constraintWeights [1]
   }
   n72 = "Constraint - dyn_coat_fr_01_l (P)"(ComponentPose=@n71, Alpha=1, ConstraintWeights_0=1) {
    alphaBoolBlend {blendOption=Linear}
    boneToModify {boneName=dyn_coat_fr_01_l}
    constraintSetup [{targetBone {boneName=thigh_l}}]
    constraintWeights [1]
   }
   n73 = "Constraint - dyn_coat_bk_01_l (P)"(ComponentPose=@n72, Alpha=0, ConstraintWeights_0=1) {
    alphaBoolBlend {blendOption=Linear}
    boneToModify {boneName=dyn_coat_bk_01_l}
    constraintSetup [{targetBone {boneName=thigh_l}}]
    constraintWeights [1]
   }
   n74 = "Constraint - dyn_coat_01_r (P)"(ComponentPose=@n73, Alpha=1, ConstraintWeights_0=1) {
    alphaBoolBlend {blendOption=Linear}
    boneToModify {boneName=dyn_coat_01_r}
    constraintSetup [{targetBone {boneName=thigh_r}}]
    constraintWeights [1]
   }
   n75 = "Constraint - dyn_coat_fr_01_r (P)"(ComponentPose=@n74, Alpha=1, ConstraintWeights_0=1) {
    alphaBoolBlend {blendOption=Linear}
    boneToModify {boneName=dyn_coat_fr_01_r}
    constraintSetup [{targetBone {boneName=thigh_r}}]
    constraintWeights [1]
   }
   n76 = "Constraint - dyn_coat_bk_01_r (P)"(ComponentPose=@n75, Alpha=0, ConstraintWeights_0=1) {
    alphaBoolBlend {blendOption=Linear}
    boneToModify {boneName=dyn_coat_bk_01_r}
    constraintSetup [{targetBone {boneName=thigh_r}}]
    constraintWeights [1]
   }
   n77 = Get lod2movingAndCrouchAlpha() -> (lod2movingAndCrouchAlpha:double)
   n78 = "Transform (Modify) Bone - Bone: dyn_coat_bk_01"(ComponentPose=@n76, Alpha=@n77, Translation="0,0,0", Rotation="0,0,-35", Scale="1,1,1") {
    rotationMode=BMM_Additive
    alphaBoolBlend {blendOption=Linear}
    alphaScaleBiasClamp {bInterpResult=true}
    boneToModify {boneName=dyn_coat_bk_01}
    rotation {roll=-35}
   }
   n79 = "Transform (Modify) Bone - Bone: dyn_coat_bk_01_r"(ComponentPose=@n78, Alpha=@n77, Translation="0,0,0", Rotation="0,0,-35", Scale="1,1,1") {
    rotationMode=BMM_Additive
    alphaBoolBlend {blendOption=Linear}
    alphaScaleBiasClamp {bInterpResult=true}
    boneToModify {boneName=dyn_coat_bk_01_r}
    rotation {roll=-35}
   }
   n80 = "Transform (Modify) Bone - Bone: dyn_coat_bk_01_l"(ComponentPose=@n79, Alpha=@n77, Translation="0,0,0", Rotation="0,0,-35", Scale="1,1,1") {
    rotationMode=BMM_Additive
    alphaBoolBlend {blendOption=Linear}
    alphaScaleBiasClamp {bInterpResult=true}
    boneToModify {boneName=dyn_coat_bk_01_l}
    rotation {roll=-35}
   }
   n81 = Trail controller - Bone: dyn_coat_bk_06_l!(ComponentPose=@n80, Alpha=1, RelaxationSpeedScale:float=1, FakeVelocity:Vector="0,0,0") -> (Pose) {
    bActorSpaceFakeVel=true
    bLimitRotation=true
    bUsePlanarLimit=true
    chainLength=4
    debugLifeTime=0.1
    maxDeltaTime=0.0333
    relaxationSpeedScale=1.2
    alphaScaleBiasClamp {bInterpResult=true, interpSpeedDecreasing=25, interpSpeedIncreasing=25}
    baseJoint {boneName=root}
    planarLimits [
     {drivingBone {boneName=root}, planeTransform {location=<0 0 10>, rotation=<rot 0 0 0>, scale=<1 1 1>}}
     {drivingBone {boneName=spine_04}, planeTransform {location=<0 40 0>, rotation=<rot 0 0 -90>, scale=<1 1 1>}}
    ]
    rotationLimits [
     {limitMax=<0 0 30>, limitMin=<0 -15 -30>}
     {limitMax=<0 0 30>, limitMin=<0 -15 -30>}
     {limitMax=<0 0 30>, limitMin=<0 -15 -30>}
     {limitMax=<180 180 180>}
    ]
    rotationOffsets [<0 0 0>, <0 0 0>, <0 0 0>, <0 0 0>]
    trailBone {boneName=dyn_coat_bk_06_l}
    trailRelaxationSpeed {editorCurveData {keys [<at 0 15 RCIM_Linear>,<at 0.5289 13.1838 RCIM_Linear>,<at 1 6 RCIM_Linear>]}}
   }
   n82 = Trail controller - Bone: dyn_coat_bk_06!(ComponentPose=@n81, Alpha=1, RelaxationSpeedScale:float=1, FakeVelocity:Vector="0,0,0") -> (Pose) {
    bActorSpaceFakeVel=true
    bLimitRotation=true
    bUsePlanarLimit=true
    chainLength=4
    debugLifeTime=0.1
    maxDeltaTime=0.0333
    relaxationSpeedScale=1.2
    alphaScaleBiasClamp {bInterpResult=true, interpSpeedDecreasing=25, interpSpeedIncreasing=25}
    baseJoint {boneName=root}
    planarLimits [
     {drivingBone {boneName=root}, planeTransform {location=<0 0 10>, rotation=<rot 0 0 0>, scale=<1 1 1>}}
     {drivingBone {boneName=spine_04}, planeTransform {location=<0 40 0>, rotation=<rot 0 0 -90>, scale=<1 1 1>}}
    ]
    rotationLimits [
     {limitMax=<0 0 30>, limitMin=<0 -15 -30>}
     {limitMax=<0 0 30>, limitMin=<0 -15 -30>}
     {limitMax=<0 0 30>, limitMin=<0 -15 -30>}
     {limitMax=<180 180 180>}
    ]
    rotationOffsets [<0 0 0>, <0 0 0>, <0 0 0>, <0 0 0>]
    trailBone {boneName=dyn_coat_bk_06}
    trailRelaxationSpeed {editorCurveData {keys [<at 0 15 RCIM_Linear>,<at 0.5289 13.1838 RCIM_Linear>,<at 1 6 RCIM_Linear>]}}
   }
   n83 = Trail controller - Bone: dyn_coat_bk_06_r!(ComponentPose=@n82, Alpha=1, RelaxationSpeedScale:float=1, FakeVelocity:Vector="0,0,0") -> (Pose) {
    bActorSpaceFakeVel=true
    bLimitRotation=true
    bUsePlanarLimit=true
    chainLength=4
    debugLifeTime=0.1
    maxDeltaTime=0.0333
    relaxationSpeedScale=1.2
    alphaScaleBiasClamp {bInterpResult=true, interpSpeedDecreasing=25, interpSpeedIncreasing=25}
    baseJoint {boneName=root}
    planarLimits [
     {drivingBone {boneName=root}, planeTransform {location=<0 0 10>, rotation=<rot 0 0 0>, scale=<1 1 1>}}
     {drivingBone {boneName=spine_04}, planeTransform {location=<0 40 0>, rotation=<rot 0 0 -90>, scale=<1 1 1>}}
    ]
    rotationLimits [
     {limitMax=<0 0 30>, limitMin=<0 -15 -30>}
     {limitMax=<0 0 30>, limitMin=<0 -15 -30>}
     {limitMax=<0 0 30>, limitMin=<0 -15 -30>}
     {limitMax=<180 180 180>}
    ]
    rotationOffsets [<0 0 0>, <0 0 0>, <0 0 0>, <0 0 0>]
    trailBone {boneName=dyn_coat_bk_06_r}
    trailRelaxationSpeed {editorCurveData {keys [<at 0 15 RCIM_Linear>,<at 0.5289 13.1838 RCIM_Linear>,<at 1 6 RCIM_Linear>]}}
   }
   Save cached pose 'trailControlCache'(Pose=Component To Local(ComponentPose=@n83)) {cacheName=trailControlCache}
   n84 = Get headTwistAngleAlpha_L() -> (headTwistAngleAlpha_L:double)
   n85 = "Transform (Modify) Bone - Bone: dyn_hoodie_fr_01_l_MSTR"(
    ComponentPose=Local To Component(LocalPose=Copy Pose From Mesh() {bUseAttachedParent=true})
    Alpha=@n84
    Translation="0,0,0"
    Rotation="-45,0,0"
    Scale="1,1,1"
   ) {
    rotationMode=BMM_Additive
    rotationSpace=BCS_BoneSpace
    alphaBoolBlend {blendOption=Linear}
    boneToModify {boneName=dyn_hoodie_fr_01_l_MSTR}
    rotation {pitch=-45}
   }
   n86 = "Transform (Modify) Bone - Bone: dyn_hoodie_01_l_MSTR"(ComponentPose=@n85, Alpha=@n84, Translation="0,0,0", Rotation="-15,0.000,0.000", Scale="1,1,1") {
    rotationMode=BMM_Additive
    rotationSpace=BCS_BoneSpace
    alphaBoolBlend {blendOption=Linear}
    boneToModify {boneName=dyn_hoodie_01_l_MSTR}
    rotation {pitch=-15}
   }
   n87 = "Transform (Modify) Bone - Bone: dyn_hoodie_bk_01_l_MSTR"(ComponentPose=@n86, Alpha=@n84, Translation="0,0,0", Rotation="-15,0.000,0.000", Scale="1,1,1") {
    rotationMode=BMM_Additive
    rotationSpace=BCS_BoneSpace
    alphaBoolBlend {blendOption=Linear}
    boneToModify {boneName=dyn_hoodie_bk_01_l_MSTR}
    rotation {pitch=-15}
   }
   Save cached pose 'initCache'(Pose=Component To Local(ComponentPose=@n87)) {cacheName=initCache}
   n88 = RigidBody(
    ComponentPose=Local To Component(LocalPose=Use cached pose 'presimCache'())
    OverrideWorldGravity=Get finalGravity() -> (finalGravity:Vector)
   ) {
    bClampLinearTranslationLimitToRefPose=true
    bOverrideWorldGravity=true
    overridePhysicsAsset=&/M_MED_Cyclone_Space_Physics_PERF2
    simulationSpace=BaseBoneSpace
    alphaBoolBlend {blendOption=Linear}
    baseBoneRef {boneName=pelvis}
    componentAppliedLinearAccClamp=<3000 3000 3000>
    componentLinearAccScale=<0.3 0.3 0.3>
    componentLinearVelScale=<0.3 0.3 0.3>
    overrideWorldGravity {z=-980}
   }
   n89 = RigidBody(ComponentPose=@n88, OverrideWorldGravity=@n55) {
    bClampLinearTranslationLimitToRefPose=true
    bOverrideWorldGravity=true
    overridePhysicsAsset=/BRCosmetics/Characters/Player/Male/Medium/Bodies/M_MED_Cyclone/Meshes/M_MED_Cyclone_necklace_Physics
    simulationSpace=BaseBoneSpace
    alphaBoolBlend {blendOption=Linear}
    baseBoneRef {boneName=pelvis}
    componentAppliedLinearAccClamp=<3000 3000 3000>
    componentLinearAccScale=<0.1 0.1 0.7>
    componentLinearVelScale=<0.3 0.3 0.7>
    overrideWorldGravity {z=-980}
   }
   n90 = "Transform (Modify) Bone - Bone: dyn_necklace_01"(ComponentPose=@n89, Alpha=1, Translation="0.005,-0.000,1.996", Rotation="-3,0.000,0.000", Scale="1,1,1") {
    rotationMode=BMM_Additive
    rotationSpace=BCS_BoneSpace
    translationMode=BMM_Additive
    translationSpace=BCS_BoneSpace
    alphaBoolBlend {blendOption=Linear}
    boneToModify {boneName=dyn_necklace_01}
    rotation {pitch=-3}
    translation {x=0.005, z=1.996}
   }
   n91 = "Transform (Modify) Bone - Bone: dyn_necklace_02"(ComponentPose=@n90, Alpha=1, Translation="0,0,0", Rotation="-10.000,0.004,0.011", Scale="1,1,1") {
    rotationMode=BMM_Additive
    rotationSpace=BCS_BoneSpace
    translationMode=BMM_Additive
    translationSpace=BCS_BoneSpace
    alphaBoolBlend {blendOption=Linear}
    boneToModify {boneName=dyn_necklace_02}
    rotation=<rot -10 0.004 0.011>
   }
   n92 = "Transform (Modify) Bone - Bone: dyn_necklace_03"(ComponentPose=@n91, Alpha=1, Translation="0,0,0", Rotation="20.000,0.007,0.006", Scale="1,1,1") {
    rotationMode=BMM_Additive
    rotationSpace=BCS_BoneSpace
    translationMode=BMM_Additive
    translationSpace=BCS_BoneSpace
    alphaBoolBlend {blendOption=Linear}
    boneToModify {boneName=dyn_necklace_03}
    rotation=<rot 20 0.007 0.006>
   }
   n93 = "Transform (Modify) Bone - Bone: dyn_necklace_03"(
    ComponentPose=@n92
    Alpha=Get isSkydiveAlpha() -> (isSkydiveAlpha:double)
    Translation="0.000,-0.000,21.674"
    Rotation="15,0.000,0.000"
    Scale="1,1,1"
   ) {
    rotationMode=BMM_Additive
    rotationSpace=BCS_BoneSpace
    translationMode=BMM_Additive
    translationSpace=BCS_BoneSpace
    alphaBoolBlend {blendOption=Linear}
    boneToModify {boneName=dyn_necklace_03}
    rotation {pitch=15}
    translation {z=21.674}
   }
   n94 = "Transform (Modify) Bone - Bone: dyn_necklace_04"(ComponentPose=@n93, Alpha=1, Translation="0,0,1.4", Rotation="10,0,0", Scale="1,1,1") {
    rotationMode=BMM_Additive
    rotationSpace=BCS_BoneSpace
    translationMode=BMM_Additive
    translationSpace=BCS_BoneSpace
    alphaBoolBlend {blendOption=Linear}
    boneToModify {boneName=dyn_necklace_04}
    rotation {pitch=10}
    translation {z=1.4}
   }
   n95 = "Transform (Modify) Bone - Bone: dyn_necklace_tp_01"(ComponentPose=@n94, Alpha=1, Translation="0,0,1", Rotation="0,0,0", Scale="1,1,1") {
    rotationMode=BMM_Additive
    rotationSpace=BCS_BoneSpace
    translationMode=BMM_Additive
    translationSpace=BCS_BoneSpace
    alphaBoolBlend {blendOption=Linear}
    boneToModify {boneName=dyn_necklace_tp_01}
    translation {z=1}
   }
   n96 = "Transform (Modify) Bone - Bone: dyn_necklace_tp_04"(ComponentPose=@n95, Alpha=1, Translation="0,0,1", Rotation="0.000,-0.035,-0.015", Scale="1,1,1") {
    rotationMode=BMM_Additive
    rotationSpace=BCS_BoneSpace
    translationMode=BMM_Additive
    translationSpace=BCS_BoneSpace
    alphaBoolBlend {blendOption=Linear}
    boneToModify {boneName=dyn_necklace_tp_04}
    rotation {roll=-0.015, yaw=-0.035}
    translation {z=1}
   }
   n97 = "Transform (Modify) Bone - Bone: dyn_necklace_tp_05"(ComponentPose=@n96, Alpha=1, Translation="0,0,0", Rotation="5.000,0.000,0.002", Scale="1,1,1") {
    rotationMode=BMM_Additive
    rotationSpace=BCS_BoneSpace
    translationMode=BMM_Additive
    translationSpace=BCS_BoneSpace
    alphaBoolBlend {blendOption=Linear}
    boneToModify {boneName=dyn_necklace_tp_05}
    rotation=<rot 5 0 0.002>
   }
   n98 = Get rotCoat_0() -> (rotCoat_0:Rotator)
   n99 = "Transform (Modify) Bone - Bone: dyn_coat_bk_00"(ComponentPose=@n97, Alpha=1, Translation="0,0,0", Rotation=@n98, Scale="1,1,1") {
    rotationMode=BMM_Additive
    rotationSpace=BCS_BoneSpace
    alphaBoolBlend {blendOption=Linear}
    alphaScaleBiasClamp {bInterpResult=true}
    boneToModify {boneName=dyn_coat_bk_00}
   }
   n100 = "Transform (Modify) Bone - Bone: dyn_coat_bk_00_l"(ComponentPose=@n99, Alpha=0.5, Translation="0,0,0", Rotation=@n98, Scale="1,1,1") {
    rotationMode=BMM_Additive
    rotationSpace=BCS_BoneSpace
    alphaBoolBlend {blendOption=Linear}
    alphaScaleBiasClamp {bInterpResult=true}
    boneToModify {boneName=dyn_coat_bk_00_l}
   }
   n101 = "Transform (Modify) Bone - Bone: dyn_coat_00_l"(ComponentPose=@n100, Alpha=0, Translation="0,0,0", Rotation=@n98, Scale="1,1,1") {
    rotationMode=BMM_Additive
    rotationSpace=BCS_BoneSpace
    alphaBoolBlend {blendOption=Linear}
    alphaScaleBiasClamp {bInterpResult=true}
    boneToModify {boneName=dyn_coat_00_l}
   }
   n102 = "Transform (Modify) Bone - Bone: dyn_coat_fr_00"(ComponentPose=@n101, Alpha=0, Translation="0,0,0", Rotation=@n98, Scale="1,1,1") {
    rotationMode=BMM_Additive
    rotationSpace=BCS_BoneSpace
    alphaBoolBlend {blendOption=Linear}
    alphaScaleBiasClamp {bInterpResult=true}
    boneToModify {boneName=dyn_coat_fr_00}
   }
   n103 = Get rotCoat_1() -> (rotCoat_1:Rotator)
   n104 = "Transform (Modify) Bone - Bone: dyn_coat_bk_00"(ComponentPose=@n102, Alpha=1, Translation="0,0,0", Rotation=@n103, Scale="1,1,1") {
    rotationMode=BMM_Additive
    rotationSpace=BCS_BoneSpace
    alphaBoolBlend {blendOption=Linear}
    alphaScaleBiasClamp {bInterpResult=true}
    boneToModify {boneName=dyn_coat_bk_00}
   }
   n105 = Get rotCoat_2() -> (rotCoat_2:Rotator)
   n106 = "Transform (Modify) Bone - Bone: dyn_coat_bk_01"(ComponentPose=@n104, Alpha=1, Translation="0,0,0", Rotation=@n105, Scale="1,1,1") {
    rotationMode=BMM_Additive
    rotationSpace=BCS_BoneSpace
    alphaBoolBlend {blendOption=Linear}
    alphaScaleBiasClamp {bInterpResult=true}
    boneToModify {boneName=dyn_coat_bk_01}
   }
   n107 = Get rotCoat_3() -> (rotCoat_3:Rotator)
   n108 = "Transform (Modify) Bone - Bone: dyn_coat_bk_02"(ComponentPose=@n106, Alpha=1, Translation="0,0,0", Rotation=@n107, Scale="1,1,1") {
    rotationMode=BMM_Additive
    rotationSpace=BCS_BoneSpace
    alphaBoolBlend {blendOption=Linear}
    alphaScaleBiasClamp {bInterpResult=true}
    boneToModify {boneName=dyn_coat_bk_02}
   }
   n109 = Get rotCoat_4() -> (rotCoat_4:Rotator)
   n110 = "Transform (Modify) Bone - Bone: dyn_coat_bk_03"(ComponentPose=@n108, Alpha=1, Translation="0,0,0", Rotation=@n109, Scale="1,1,1") {
    rotationMode=BMM_Additive
    rotationSpace=BCS_BoneSpace
    alphaBoolBlend {blendOption=Linear}
    alphaScaleBiasClamp {bInterpResult=true}
    boneToModify {boneName=dyn_coat_bk_03}
   }
   n111 = Get rotCoat_5() -> (rotCoat_5:Rotator)
   n112 = "Transform (Modify) Bone - Bone: dyn_coat_bk_04"(ComponentPose=@n110, Alpha=1, Translation="0,0,0", Rotation=@n111, Scale="1,1,1") {
    rotationMode=BMM_Additive
    rotationSpace=BCS_BoneSpace
    alphaBoolBlend {blendOption=Linear}
    alphaScaleBiasClamp {bInterpResult=true}
    boneToModify {boneName=dyn_coat_bk_04}
   }
   n113 = Get rotCoat_6() -> (rotCoat_6:Rotator)
   n114 = "Transform (Modify) Bone - Bone: dyn_coat_bk_05"(ComponentPose=@n112, Alpha=1, Translation="0,0,0", Rotation=@n113, Scale="1,1,1") {
    rotationMode=BMM_Additive
    rotationSpace=BCS_BoneSpace
    alphaBoolBlend {blendOption=Linear}
    alphaScaleBiasClamp {bInterpResult=true}
    boneToModify {boneName=dyn_coat_bk_05}
   }
   n115 = "Transform (Modify) Bone - Bone: dyn_coat_bk_00_l"(ComponentPose=@n114, Alpha=0, Translation="0,0,0", Rotation=@n103, Scale="1,1,1") {
    rotationMode=BMM_Additive
    rotationSpace=BCS_BoneSpace
    alphaBoolBlend {blendOption=Linear}
    alphaScaleBiasClamp {bInterpResult=true}
    boneToModify {boneName=dyn_coat_bk_00_l}
   }
   n116 = "Transform (Modify) Bone - Bone: dyn_coat_bk_01_l"(ComponentPose=@n115, Alpha=0.5, Translation="0,0,0", Rotation=@n105, Scale="1,1,1") {
    rotationMode=BMM_Additive
    rotationSpace=BCS_BoneSpace
    alphaBoolBlend {blendOption=Linear}
    alphaScaleBiasClamp {bInterpResult=true}
    boneToModify {boneName=dyn_coat_bk_01_l}
   }
   n117 = "Transform (Modify) Bone - Bone: dyn_coat_bk_02_l"(ComponentPose=@n116, Alpha=0.5, Translation="0,0,0", Rotation=@n107, Scale="1,1,1") {
    rotationMode=BMM_Additive
    rotationSpace=BCS_BoneSpace
    alphaBoolBlend {blendOption=Linear}
    alphaScaleBiasClamp {bInterpResult=true}
    boneToModify {boneName=dyn_coat_bk_02_l}
   }
   n118 = "Transform (Modify) Bone - Bone: dyn_coat_bk_03_l"(ComponentPose=@n117, Alpha=0.5, Translation="0,0,0", Rotation=@n109, Scale="1,1,1") {
    rotationMode=BMM_Additive
    rotationSpace=BCS_BoneSpace
    alphaBoolBlend {blendOption=Linear}
    alphaScaleBiasClamp {bInterpResult=true}
    boneToModify {boneName=dyn_coat_bk_03_l}
   }
   n119 = "Transform (Modify) Bone - Bone: dyn_coat_bk_04_l"(ComponentPose=@n118, Alpha=0.5, Translation="0,0,0", Rotation=@n111, Scale="1,1,1") {
    rotationMode=BMM_Additive
    rotationSpace=BCS_BoneSpace
    alphaBoolBlend {blendOption=Linear}
    alphaScaleBiasClamp {bInterpResult=true}
    boneToModify {boneName=dyn_coat_bk_04_l}
   }
   n120 = "Transform (Modify) Bone - Bone: dyn_coat_bk_05_l"(ComponentPose=@n119, Alpha=0.5, Translation="0,0,0", Rotation=@n113, Scale="1,1,1") {
    rotationMode=BMM_Additive
    rotationSpace=BCS_BoneSpace
    alphaBoolBlend {blendOption=Linear}
    alphaScaleBiasClamp {bInterpResult=true}
    boneToModify {boneName=dyn_coat_bk_05_l}
   }
   n121 = "Transform (Modify) Bone - Bone: dyn_coat_00_l"(ComponentPose=@n120, Alpha=0, Translation="0,0,0", Rotation=@n103, Scale="1,1,1") {
    rotationMode=BMM_Additive
    rotationSpace=BCS_BoneSpace
    alphaBoolBlend {blendOption=Linear}
    alphaScaleBiasClamp {bInterpResult=true}
    boneToModify {boneName=dyn_coat_00_l}
   }
   n122 = "Transform (Modify) Bone - Bone: dyn_coat_01_l"(ComponentPose=@n121, Alpha=0, Translation="0,0,0", Rotation=@n105, Scale="1,1,1") {
    rotationMode=BMM_Additive
    rotationSpace=BCS_BoneSpace
    alphaBoolBlend {blendOption=Linear}
    alphaScaleBiasClamp {bInterpResult=true}
    boneToModify {boneName=dyn_coat_01_l}
   }
   n123 = "Transform (Modify) Bone - Bone: dyn_coat_02_l"(ComponentPose=@n122, Alpha=0, Translation="0,0,0", Rotation=@n107, Scale="1,1,1") {
    rotationMode=BMM_Additive
    rotationSpace=BCS_BoneSpace
    alphaBoolBlend {blendOption=Linear}
    alphaScaleBiasClamp {bInterpResult=true}
    boneToModify {boneName=dyn_coat_02_l}
   }
   n124 = "Transform (Modify) Bone - Bone: dyn_coat_03_l"(ComponentPose=@n123, Alpha=0, Translation="0,0,0", Rotation=@n109, Scale="1,1,1") {
    rotationMode=BMM_Additive
    rotationSpace=BCS_BoneSpace
    alphaBoolBlend {blendOption=Linear}
    alphaScaleBiasClamp {bInterpResult=true}
    boneToModify {boneName=dyn_coat_03_l}
   }
   n125 = "Transform (Modify) Bone - Bone: dyn_coat_04_l"(ComponentPose=@n124, Alpha=0, Translation="0,0,0", Rotation=@n111, Scale="1,1,1") {
    rotationMode=BMM_Additive
    rotationSpace=BCS_BoneSpace
    alphaBoolBlend {blendOption=Linear}
    alphaScaleBiasClamp {bInterpResult=true}
    boneToModify {boneName=dyn_coat_04_l}
   }
   n126 = "Transform (Modify) Bone - Bone: dyn_coat_05_l"(ComponentPose=@n125, Alpha=0, Translation="0,0,0", Rotation=@n113, Scale="1,1,1") {
    rotationMode=BMM_Additive
    rotationSpace=BCS_BoneSpace
    alphaBoolBlend {blendOption=Linear}
    alphaScaleBiasClamp {bInterpResult=true}
    boneToModify {boneName=dyn_coat_05_l}
   }
   n127 = "Transform (Modify) Bone - Bone: dyn_coat_fr_00"(ComponentPose=@n126, Alpha=0, Translation="0,0,0", Rotation=@n103, Scale="1,1,1") {
    rotationMode=BMM_Additive
    rotationSpace=BCS_BoneSpace
    alphaBoolBlend {blendOption=Linear}
    alphaScaleBiasClamp {bInterpResult=true}
    boneToModify {boneName=dyn_coat_fr_00}
   }
   n128 = "Transform (Modify) Bone - Bone: dyn_coat_fr_01_l"!(ComponentPose=@n127, Alpha:float=0, Translation="0,0,0", Rotation=@n105, Scale="1,1,1") -> (Pose) {
    rotationMode=BMM_Additive
    rotationSpace=BCS_BoneSpace
    alphaBoolBlend {blendOption=Linear}
    alphaScaleBiasClamp {bInterpResult=true}
    boneToModify {boneName=dyn_coat_fr_01_l}
   }
   n129 = "Transform (Modify) Bone - Bone: dyn_coat_fr_02_l"(ComponentPose=@n128, Alpha=0, Translation="0,0,0", Rotation=@n107, Scale="1,1,1") {
    rotationMode=BMM_Additive
    rotationSpace=BCS_BoneSpace
    alphaBoolBlend {blendOption=Linear}
    alphaScaleBiasClamp {bInterpResult=true}
    boneToModify {boneName=dyn_coat_fr_02_l}
   }
   n130 = "Transform (Modify) Bone - Bone: dyn_coat_fr_03_l"(ComponentPose=@n129, Alpha=0, Translation="0,0,0", Rotation=@n109, Scale="1,1,1") {
    rotationMode=BMM_Additive
    rotationSpace=BCS_BoneSpace
    alphaBoolBlend {blendOption=Linear}
    alphaScaleBiasClamp {bInterpResult=true}
    boneToModify {boneName=dyn_coat_fr_03_l}
   }
   n131 = "Transform (Modify) Bone - Bone: dyn_coat_fr_04_l"(ComponentPose=@n130, Alpha=0, Translation="0,0,0", Rotation=@n111, Scale="1,1,1") {
    rotationMode=BMM_Additive
    rotationSpace=BCS_BoneSpace
    alphaBoolBlend {blendOption=Linear}
    alphaScaleBiasClamp {bInterpResult=true}
    boneToModify {boneName=dyn_coat_fr_04_l}
   }
   n132 = "Transform (Modify) Bone - Bone: dyn_coat_fr_05_l"(ComponentPose=@n131, Alpha=0, Translation="0,0,0", Rotation=@n113, Scale="1,1,1") {
    rotationMode=BMM_Additive
    rotationSpace=BCS_BoneSpace
    alphaBoolBlend {blendOption=Linear}
    alphaScaleBiasClamp {bInterpResult=true}
    boneToModify {boneName=dyn_coat_fr_05_l}
   }
   n133 = Get rotCoat_0()
   n134 = "Transform (Modify) Bone - Bone: dyn_coat_bk_00_r"(ComponentPose=@n132, Alpha=1, Translation="0,0,0", Rotation=@n133, Scale="1,1,1") {
    rotationMode=BMM_Additive
    rotationSpace=BCS_BoneSpace
    alphaBoolBlend {blendOption=Linear}
    alphaScaleBiasClamp {bInterpResult=true}
    boneToModify {boneName=dyn_coat_bk_00_r}
   }
   n135 = "Transform (Modify) Bone - Bone: dyn_coat_00_r"(ComponentPose=@n134, Alpha=1, Translation="0,0,0", Rotation=@n133, Scale="1,1,1") {
    rotationMode=BMM_Additive
    rotationSpace=BCS_BoneSpace
    alphaBoolBlend {blendOption=Linear}
    alphaScaleBiasClamp {bInterpResult=true}
    boneToModify {boneName=dyn_coat_00_r}
   }
   n136 = "Transform (Modify) Bone - Bone: dyn_coat_fr_00_r"(ComponentPose=@n135, Alpha=0, Translation="0,0,0", Rotation=@n133, Scale="1,1,1") {
    rotationMode=BMM_Additive
    rotationSpace=BCS_BoneSpace
    alphaBoolBlend {blendOption=Linear}
    alphaScaleBiasClamp {bInterpResult=true}
    boneToModify {boneName=dyn_coat_fr_00_r}
   }
   n137 = Get rotCoat_1()
   n138 = "Transform (Modify) Bone - Bone: dyn_coat_bk_00_r"(ComponentPose=@n136, Alpha=1, Translation="0,0,0", Rotation=@n137, Scale="1,1,1") {
    rotationMode=BMM_Additive
    rotationSpace=BCS_BoneSpace
    alphaBoolBlend {blendOption=Linear}
    alphaScaleBiasClamp {bInterpResult=true}
    boneToModify {boneName=dyn_coat_bk_00_r}
   }
   n139 = Get rotCoat_2()
   n140 = "Transform (Modify) Bone - Bone: dyn_coat_01_r"(ComponentPose=@n138, Alpha=1, Translation="0,0,0", Rotation=@n139, Scale="1,1,1") {
    rotationMode=BMM_Additive
    rotationSpace=BCS_BoneSpace
    alphaBoolBlend {blendOption=Linear}
    alphaScaleBiasClamp {bInterpResult=true}
    boneToModify {boneName=dyn_coat_01_r}
   }
   n141 = Get rotCoat_3()
   n142 = "Transform (Modify) Bone - Bone: dyn_coat_bk_02_r"(ComponentPose=@n140, Alpha=1, Translation="0,0,0", Rotation=@n141, Scale="1,1,1") {
    rotationMode=BMM_Additive
    rotationSpace=BCS_BoneSpace
    alphaBoolBlend {blendOption=Linear}
    alphaScaleBiasClamp {bInterpResult=true}
    boneToModify {boneName=dyn_coat_bk_02_r}
   }
   n143 = "Transform (Modify) Bone - Bone: dyn_coat_bk_03_r"(ComponentPose=@n142, Alpha=1, Translation="0,0,0", Rotation=Get rotCoat_foot_0() -> (rotCoat_foot_0:Rotator), Scale="1,1,1") {
    rotationMode=BMM_Additive
    rotationSpace=BCS_BoneSpace
    alphaBoolBlend {blendOption=Linear}
    alphaScaleBiasClamp {bInterpResult=true}
    boneToModify {boneName=dyn_coat_bk_03_r}
   }
   n144 = "Transform (Modify) Bone - Bone: dyn_coat_bk_04_r"(ComponentPose=@n143, Alpha=1, Translation="0,0,0", Rotation=Get rotCoat_foot_1() -> (rotCoat_foot_1:Rotator), Scale="1,1,1") {
    rotationMode=BMM_Additive
    rotationSpace=BCS_BoneSpace
    alphaBoolBlend {blendOption=Linear}
    alphaScaleBiasClamp {bInterpResult=true}
    boneToModify {boneName=dyn_coat_bk_04_r}
   }
   n145 = "Transform (Modify) Bone - Bone: dyn_coat_bk_05_r"(ComponentPose=@n144, Alpha=1, Translation="0,0,0", Rotation=Get rotCoat_foot_2() -> (rotCoat_foot_2:Rotator), Scale="1,1,1") {
    rotationMode=BMM_Additive
    rotationSpace=BCS_BoneSpace
    alphaBoolBlend {blendOption=Linear}
    alphaScaleBiasClamp {bInterpResult=true}
    boneToModify {boneName=dyn_coat_bk_05_r}
   }
   n146 = "Transform (Modify) Bone - Bone: dyn_coat_00_r"(ComponentPose=@n145, Alpha=1, Translation="0,0,0", Rotation=@n137, Scale="1,1,1") {
    rotationMode=BMM_Additive
    rotationSpace=BCS_BoneSpace
    alphaBoolBlend {blendOption=Linear}
    alphaScaleBiasClamp {bInterpResult=true}
    boneToModify {boneName=dyn_coat_00_r}
   }
   n147 = "Transform (Modify) Bone - Bone: dyn_coat_01_r"(ComponentPose=@n146, Alpha=1, Translation="0,0,0", Rotation=@n139, Scale="1,1,1") {
    rotationMode=BMM_Additive
    rotationSpace=BCS_BoneSpace
    alphaBoolBlend {blendOption=Linear}
    alphaScaleBiasClamp {bInterpResult=true}
    boneToModify {boneName=dyn_coat_01_r}
   }
   n148 = "Transform (Modify) Bone - Bone: dyn_coat_02_r"(ComponentPose=@n147, Alpha=1, Translation="0,0,0", Rotation=@n141, Scale="1,1,1") {
    rotationMode=BMM_Additive
    rotationSpace=BCS_BoneSpace
    alphaBoolBlend {blendOption=Linear}
    alphaScaleBiasClamp {bInterpResult=true}
    boneToModify {boneName=dyn_coat_02_r}
   }
   n149 = Get rotCoat_4()
   n150 = "Transform (Modify) Bone - Bone: dyn_coat_03_r"(ComponentPose=@n148, Alpha=1, Translation="0,0,0", Rotation=@n149, Scale="1,1,1") {
    rotationMode=BMM_Additive
    rotationSpace=BCS_BoneSpace
    alphaBoolBlend {blendOption=Linear}
    alphaScaleBiasClamp {bInterpResult=true}
    boneToModify {boneName=dyn_coat_03_r}
   }
   n151 = Get rotCoat_5()
   n152 = "Transform (Modify) Bone - Bone: dyn_coat_04_r"(ComponentPose=@n150, Alpha=1, Translation="0,0,0", Rotation=@n151, Scale="1,1,1") {
    rotationMode=BMM_Additive
    rotationSpace=BCS_BoneSpace
    alphaBoolBlend {blendOption=Linear}
    alphaScaleBiasClamp {bInterpResult=true}
    boneToModify {boneName=dyn_coat_04_r}
   }
   n153 = Get rotCoat_6()
   n154 = "Transform (Modify) Bone - Bone: dyn_coat_05_r"(ComponentPose=@n152, Alpha=1, Translation="0,0,0", Rotation=@n153, Scale="1,1,1") {
    rotationMode=BMM_Additive
    rotationSpace=BCS_BoneSpace
    alphaBoolBlend {blendOption=Linear}
    alphaScaleBiasClamp {bInterpResult=true}
    boneToModify {boneName=dyn_coat_05_r}
   }
   n155 = "Transform (Modify) Bone - Bone: dyn_coat_fr_00_r"(ComponentPose=@n154, Alpha=1, Translation="0,0,0", Rotation=@n137, Scale="1,1,1") {
    rotationMode=BMM_Additive
    rotationSpace=BCS_BoneSpace
    alphaBoolBlend {blendOption=Linear}
    alphaScaleBiasClamp {bInterpResult=true}
    boneToModify {boneName=dyn_coat_fr_00_r}
   }
   n156 = "Transform (Modify) Bone - Bone: dyn_coat_fr_01_r"(ComponentPose=@n155, Alpha=1, Translation="0,0,0", Rotation=@n139, Scale="1,1,1") {
    rotationMode=BMM_Additive
    rotationSpace=BCS_BoneSpace
    alphaBoolBlend {blendOption=Linear}
    alphaScaleBiasClamp {bInterpResult=true}
    boneToModify {boneName=dyn_coat_fr_01_r}
   }
   n157 = "Transform (Modify) Bone - Bone: dyn_coat_fr_02_r"(ComponentPose=@n156, Alpha=1, Translation="0,0,0", Rotation=@n141, Scale="1,1,1") {
    rotationMode=BMM_Additive
    rotationSpace=BCS_BoneSpace
    alphaBoolBlend {blendOption=Linear}
    alphaScaleBiasClamp {bInterpResult=true}
    boneToModify {boneName=dyn_coat_fr_02_r}
   }
   n158 = "Transform (Modify) Bone - Bone: dyn_coat_fr_03_r"(ComponentPose=@n157, Alpha=1, Translation="0,0,0", Rotation=@n149, Scale="1,1,1") {
    rotationMode=BMM_Additive
    rotationSpace=BCS_BoneSpace
    alphaBoolBlend {blendOption=Linear}
    alphaScaleBiasClamp {bInterpResult=true}
    boneToModify {boneName=dyn_coat_fr_03_r}
   }
   n159 = "Transform (Modify) Bone - Bone: dyn_coat_fr_04_r"(ComponentPose=@n158, Alpha=1, Translation="0,0,0", Rotation=@n151, Scale="1,1,1") {
    rotationMode=BMM_Additive
    rotationSpace=BCS_BoneSpace
    alphaBoolBlend {blendOption=Linear}
    alphaScaleBiasClamp {bInterpResult=true}
    boneToModify {boneName=dyn_coat_fr_04_r}
   }
   n160 = "Transform (Modify) Bone - Bone: dyn_coat_fr_05_r"(ComponentPose=@n159, Alpha=1, Translation="0,0,0", Rotation=@n153, Scale="1,1,1") {
    rotationMode=BMM_Additive
    rotationSpace=BCS_BoneSpace
    alphaBoolBlend {blendOption=Linear}
    alphaScaleBiasClamp {bInterpResult=true}
    boneToModify {boneName=dyn_coat_fr_05_r}
   }
   n161 = Get not_crouch() -> (not_crouch:bool)
   n162 = "Transform (Modify) Bone - Bone: dyn_coat_bk_02"!(ComponentPose=@n160, bAlphaBoolEnabled:bool=@n161, Translation="0,0,0", Rotation="10,0,0", Scale="1,1,1") -> (Pose) {
    alphaInputType=Bool
    rotationMode=BMM_Additive
    rotationSpace=BCS_BoneSpace
    alphaBoolBlend {blendInTime=0.5, blendOption=Linear, blendOutTime=0.5}
    boneToModify {boneName=dyn_coat_bk_02}
    rotation {pitch=10}
   }
   n163 = "Transform (Modify) Bone - Bone: dyn_coat_bk_02_l"!(ComponentPose=@n162, bAlphaBoolEnabled:bool=@n161, Translation="0,0,0", Rotation="10,0,0", Scale="1,1,1") -> (Pose) {
    alphaInputType=Bool
    rotationMode=BMM_Additive
    rotationSpace=BCS_BoneSpace
    alphaBoolBlend {blendInTime=0.5, blendOption=Linear, blendOutTime=0.5}
    boneToModify {boneName=dyn_coat_bk_02_l}
    rotation {pitch=10}
   }
   n164 = "Transform (Modify) Bone - Bone: dyn_coat_bk_02_r"!(ComponentPose=@n163, bAlphaBoolEnabled:bool=@n161, Translation="0,0,0", Rotation="10,0,0", Scale="1,1,1") -> (Pose) {
    alphaInputType=Bool
    rotationMode=BMM_Additive
    rotationSpace=BCS_BoneSpace
    alphaBoolBlend {blendInTime=0.5, blendOption=Linear, blendOutTime=0.5}
    boneToModify {boneName=dyn_coat_bk_02_r}
    rotation {pitch=10}
   }
   n165 = "Transform (Modify) Bone - Bone: dyn_coat_bk_03"!(ComponentPose=@n164, bAlphaBoolEnabled:bool=@n161, Translation="0,0,0", Rotation="15,0,0", Scale="1,1,1") -> (Pose) {
    alphaInputType=Bool
    rotationMode=BMM_Additive
    rotationSpace=BCS_BoneSpace
    alphaBoolBlend {blendInTime=0.5, blendOption=Linear, blendOutTime=0.5}
    boneToModify {boneName=dyn_coat_bk_03}
    rotation {pitch=15}
   }
   n166 = "Transform (Modify) Bone - Bone: dyn_coat_bk_03_l"!(ComponentPose=@n165, bAlphaBoolEnabled:bool=@n161, Translation="0,0,0", Rotation="12,0,0", Scale="1,1,1") -> (Pose) {
    alphaInputType=Bool
    rotationMode=BMM_Additive
    rotationSpace=BCS_BoneSpace
    alphaBoolBlend {blendInTime=0.5, blendOption=Linear, blendOutTime=0.5}
    boneToModify {boneName=dyn_coat_bk_03_l}
    rotation {pitch=12}
   }
   n167 = "Transform (Modify) Bone - Bone: dyn_coat_bk_03_r"!(ComponentPose=@n166, bAlphaBoolEnabled:bool=@n161, Translation="0,0,0", Rotation="12,0,0", Scale="1,1,1") -> (Pose) {
    alphaInputType=Bool
    rotationMode=BMM_Additive
    rotationSpace=BCS_BoneSpace
    alphaBoolBlend {blendInTime=0.5, blendOption=Linear, blendOutTime=0.5}
    boneToModify {boneName=dyn_coat_bk_03_r}
    rotation {pitch=12}
   }
   n168 = "Transform (Modify) Bone - Bone: dyn_coat_bk_04"!(ComponentPose=@n167, bAlphaBoolEnabled:bool=Get bIsCrouching(), Translation="0,0,0", Rotation="20,0,-10", Scale="1,1,1") -> (Pose) {
    alphaInputType=Bool
    rotationMode=BMM_Additive
    rotationSpace=BCS_BoneSpace
    alphaBoolBlend {blendInTime=0.5, blendOption=Linear, blendOutTime=0.5}
    boneToModify {boneName=dyn_coat_bk_04}
    rotation {pitch=20, roll=-10}
   }
   n169 = "Transform (Modify) Bone - Bone: dyn_coatupper_md_01_l"(ComponentPose=@n168, Alpha=1, Translation="-0.000,0.004,-1.877", Rotation="0,0,0", Scale="1,1,1") {
    rotationMode=BMM_Additive
    rotationSpace=BCS_BoneSpace
    translationMode=BMM_Additive
    translationSpace=BCS_BoneSpace
    alphaBoolBlend {blendOption=Linear}
    boneToModify {boneName=dyn_coatupper_md_01_l}
    translation {y=0.004, z=-1.877}
   }
   n170 = "Transform (Modify) Bone - Bone: dyn_necklace_01_MSTR"(
    ComponentPose=@n169
    Alpha=Get isCycloneDancingAlpha() -> (isCycloneDancingAlpha:double)
    Translation="0,0,0"
    Rotation="5.5,0,0"
    Scale="1,1,1"
   ) {
    rotationMode=BMM_Additive
    rotationSpace=BCS_BoneSpace
    translationMode=BMM_Additive
    translationSpace=BCS_BoneSpace
    alphaBoolBlend {blendOption=Linear}
    boneToModify {boneName=dyn_necklace_01_MSTR}
    rotation {pitch=5.5}
   }
   Save cached pose 'rbanCache'(Pose=Component To Local(ComponentPose=@n170)) {cacheName=rbanCache}
   Use cached pose 'presimCache'()
  ]}
 EventGraph {nodes [
   Event BlueprintUpdateAnimation()
   > n1 = Sequence!(execute) -> (then_0, then_1, then_2:Exec, then_3:Exec, then_4:Exec, then_5:Exec, then_6:Exec, then_7:Exec, then_8:Exec, then_9:Exec)
   > n2 = Gravity Override State Function DEPRECATED(CharacterPartAnimInstance=Self-Reference(), GameTimeInSeconds=GetGameTimeInSeconds(), GravityOverrideParams=Get Gravity Override Params())
   > n3 = Set windVector(windVector:Vector=SelectVector(A=@n2, B="0, 0,-980", bPickA=Get bIsSkyDiving())) -> (Output_Get:Vector)
   n4 = BreakTransform(InTransform=GetDeltaTransformFromRefPose(self=GetOwnerSkeletalMeshComponent(), BoneName=pelvis, BaseName=root))
   n5 = "float > float"(A="Absolute (Float)"(A=BreakRotator(InRot=@n4.Rotation).Yaw), B=90) {is=title K2Node_CallFunction}
   n6 = SelectFloat(A=980, B=-980, bPickA=@n5)
   > Set finalGravity(finalGravity:Vector=SelectVector(A=MakeVector(Y=0, Z=@n6), B=@n3, bPickA=Get bIsPlayingEmote())) -> (Output_Get:Vector)
   > Set simSpaceRoot(simSpaceRoot:double=SelectFloat(A=1, B=0, bPickA=NOT Boolean(A=@n5))) -> (Output_Get:double)
   > Set simSpaceSpine(simSpaceSpine:double=SelectFloat(A=1, B=0, bPickA=@n5)) -> (Output_Get:double)
   Set isSkydiveAlpha(execute=@n1.then_1, isSkydiveAlpha:double="To Float (Boolean)"(InBool=Get bIsSkyDiving())) -> (Output_Get:double)
   n7 = Get CurrentLOD()
   n8 = Get bRigidBodyEnabled()
   Set rigids_off(execute=@n1.then_2, rigids_off:bool=OR Boolean(A=NOT Boolean(A=@n8), B="integer > integer"(A=@n7, B=1) {is=title K2Node_CallFunction})) -> (Output_Get:bool)
   n9 = Get bIsCrouching()
   Set noiseRotRoot(execute=@n1.then_3, noiseRotRoot:Rotator=SelectRotator!(A:Rotator="65, 0, 0", B:Rotator="0, 0, 0", bPickA:bool=@n9) -> (ReturnValue:Rotator)) -> (Output_Get:Rotator)
   > Set scaleCoat(scaleCoat:Vector=SelectVector(A="0.5,1,1", B="1,1,1", bPickA=@n9)) -> (Output_Get:Vector)
   > Set scaleCoatRight(scaleCoatRight:Vector=SelectVector(A="0.2,1,1", B="1,1,1", bPickA=@n9)) -> (Output_Get:Vector)
   > Set scaleCoatMid(scaleCoatMid:Vector=SelectVector(A="0.3,1,1", B="1,1,1", bPickA=@n9)) -> (Output_Get:Vector)
   > Set coatUp(coatUp:Vector=SelectVector(A="0, 0, 0", B="0, 0, 0", bPickA=@n9)) -> (Output_Get:Vector)
   n10 = Get PawnSpeed2D()
   n11 = NOT Boolean(A="float > float"(A=@n10, B=100))
   > Set scaleCoatAlpha(scaleCoatAlpha:double=SelectFloat(A=MapRangeClamped(Value=@n10, InRangeA=0, InRangeB=280, OutRangeA=0.65, OutRangeB=0), B=0, bPickA=AND Boolean(A=@n9, B=@n11))) -> (Output_Get:double)
   Set rotCoat_0(execute=@n1.then_4, rotCoat_0:Rotator="0, 0, 0") -> (Output_Get:Rotator)
   > Set rotCoat_1(rotCoat_1:Rotator="0, 0, 0") -> (Output_Get:Rotator)
   n12 = MapRangeClamped(Value=Get PawnSpeed2D(), InRangeA=0, InRangeB=150, OutRangeA=1, OutRangeB=0.15)
   n13 = Get bIsCrouching()
   > Set rotCoat_2(rotCoat_2:Rotator="Scale Rotator (Float)"!(A:Rotator=SelectRotator(A="1, 0, 0", B="0, 0, 0", bPickA=@n13), B:float=@n12) -> (ReturnValue:Rotator)) -> (Output_Get:Rotator)
   > Set rotCoat_3(rotCoat_3:Rotator="Scale Rotator (Float)"(A=SelectRotator(A="2, 0, 0", B="0, 0, 0", bPickA=@n13), B=@n12)) -> (Output_Get:Rotator)
   > Set rotCoat_4(rotCoat_4:Rotator="Scale Rotator (Float)"(A=SelectRotator(A="4, 0, 0", B="0, 0, 0", bPickA=@n13), B=@n12)) -> (Output_Get:Rotator)
   > Set rotCoat_5(rotCoat_5:Rotator="Scale Rotator (Float)"(A=SelectRotator(A="8, 0, 0", B="0, 0, 0", bPickA=@n13), B=@n12)) -> (Output_Get:Rotator)
   > Set rotCoat_6(rotCoat_6:Rotator="Scale Rotator (Float)"(A=SelectRotator(A="16, 0, 0", B="0, 0, 0", bPickA=@n13), B=@n12)) -> (Output_Get:Rotator)
   > Set rotCoat_7(rotCoat_7:Rotator="Scale Rotator (Float)"(A=SelectRotator(A="35, 0, 0", B="0, 0, 0", bPickA=@n13), B=@n12)) -> (Output_Get:Rotator)
   > Set rotCoat_foot_0(rotCoat_foot_0:Rotator="Scale Rotator (Float)"(A=SelectRotator(A="30, 0, 0", B="0, 0, 0", bPickA=@n13), B=@n12)) -> (Output_Get:Rotator)
   > Set rotCoat_foot_2(rotCoat_foot_2:Rotator="Scale Rotator (Float)"(A=SelectRotator(A="40, 0, 0", B="0, 0, 0", bPickA=@n13), B=@n12)) -> (Output_Get:Rotator)
   > Set rotCoat_foot_3(rotCoat_foot_3:Rotator="Scale Rotator (Float)"(A=SelectRotator(A="45, 0,0", B="0, 0, 0", bPickA=@n13), B=@n12)) -> (Output_Get:Rotator)
   n14 = Get PawnSpeed2D()
   n15 = "float >= float"(A=@n14, B=150) {is=title K2Node_CallFunction}
   n16 = Get bIsCrouching()
   Set lod2movingAndCrouchAlpha(
    execute=@n1.then_6
    lod2movingAndCrouchAlpha:double=SelectFloat(A=MapRangeClamped(Value=@n14, InRangeA=0, InRangeB=200), B=0, bPickA=AND Boolean(A=@n16, B=@n15))
   ) -> (Output_Get:double)
   n17 = BreakTransform(InTransform=GetDeltaTransformFromRefPose(self=GetOwnerSkeletalMeshComponent(), BoneName=head, BaseName=neck_02))
   n18 = Set headTwistAngleAlpha_L(execute=@n1.then_7, headTwistAngleAlpha_L:double=MapRangeClamped(Value=BreakRotator(InRot=@n17.Rotation), InRangeA=-10, InRangeB=10)) -> (Output_Get:double)
   > Set headTwistAngleAlpha_R(headTwistAngleAlpha_R:double=MapRangeClamped(Value=float - float(B=@n18) {is=title K2Node_CallFunction}, InRangeA=0, InRangeB=1, OutRangeA=0.25)) -> (Output_Get:double)
   n19 = GetAnimInstance(self=GetOwnerSkeletalMeshComponent())
   Is Valid(exec=@n1.then_8, InputObject=@n19)
   n20 = Montage_IsActive!(
    self:Anim Instance Object Reference=@n19
    Montage:Anim Montage Object Reference=/BRCosmetics/Animation/Game/MainPlayer/Emotes/Cyclone/Cyclone_CMM_M
   ) -> (ReturnValue:bool)
   n21 = Montage_IsActive(self=@n19, Montage=/BRCosmetics/Animation/Game/MainPlayer/Emotes/Cyclone_Headbang/emote_cyclone_headbang_CMM_M)
   > Set isCycloneDancingAlpha(isCycloneDancingAlpha:double="To Float (Boolean)"(InBool=OR Boolean(A=@n21, B=@n20))) -> (Output_Get:double)
   Set not_crouch(execute=@n1.then_9, not_crouch:bool=NOT Boolean(A=Get bIsCrouching())) -> (Output_Get:bool)
   > n22 = Branch(Condition=OR Boolean(A=Get bIsJumping(), B=Get bIsFalling()))
   n23 = BreakTransform(InTransform=GetDeltaTransformFromRefPose(self=GetOwnerSkeletalMeshComponent(), BoneName=thigh_r, BaseName=pelvis))
   > n24 = Set jump_alpha(jump_alpha:double=MapRangeClamped(Value=BreakRotator(InRot=@n23.Rotation).Yaw, InRangeA=0, InRangeB=-70)) -> (Output_Get:double)
   n25 = Branch(execute=[@n24.then, @n27.then], Condition=Get bIsCrouching())
   n26 = BreakTransform(InTransform=GetDeltaTransformFromRefPose(self=GetOwnerSkeletalMeshComponent(), BoneName=thigh_r, BaseName=pelvis))
   > Set thigh_r_crouch_pitch_alpha(thigh_r_crouch_pitch_alpha:double=MapRangeClamped(Value=BreakRotator(InRot=@n26.Rotation).Pitch, InRangeA=0, InRangeB=40)) -> (Output_Get:double)
   Set thigh_r_crouch_pitch_alpha(execute=@n25.else, thigh_r_crouch_pitch_alpha=0)
   n27 = Set jump_alpha(execute=@n22.else, jump_alpha=0)
   Event OnEndSkydiving()
   > ResetDynamics(InTeleportType=TeleportPhysics)
   > ResetDynamics(InTeleportType=ResetPhysics)
   Event OnBeginCrouching()
   > Set crouchingAlpha(crouchingAlpha:double=0.75) -> (Output_Get:double)
   Event OnEndCrouching()
   > Set crouchingAlpha(crouchingAlpha=0.5)
   Event BlueprintInitializeAnimation()
   > Set crouchingAlpha(crouchingAlpha=0.5)
   TryGetPawnOwner()
   vector * float(
    A=Break Fort Anim Input Skydiving External Force(FortAnimInput_SkydivingExternalForce=Get SkydivingExternalForceInput()) -> (
     bUseSkydivingVectorForce=~
     SkydivingVectorForce=~
     HeadToPelvisDirection=~
     FloatingMultiplier=~
     FloatingAdditive=~
     DivingMultiplier=~
     DivingAdditive=~
     ParachutingMultiplier=~
     ParachutingAdditive=~
    ).SkydivingVectorForce
    B=0.01
   )
  ]}
]
variables [
 calf_l_crouch_yaw_alpha {type=number}
 coatUp {type=Vector}
 crouchingAlpha {type=number}
 finalGravity {type=Vector}
 Gravity Override Params {
  type=GravityOverrideParamsStruct
  default {
   skydive_FloatingDown {gravityOverride=<0 -300 120>}
   skydive_GoingUp {gravityOverride=<0 -100 -100>}
   skydive_Parachuting {gravityOverride=<0 -200 -980>}
  }
 }
 headTwistAngleAlpha_L {type=number}
 headTwistAngleAlpha_R {type=number}
 isCycloneDancingAlpha {type=number}
 isSkydiveAlpha {type=number}
 jump_alpha {type=number}
 lod2movingAndCrouchAlpha {type=number}
 noiseRotRoot {type=Rotator}
 not_crouch {type=boolean}
 rigids_off {type=boolean}
 rotCoat_0 {type=Rotator}
 rotCoat_1 {type=Rotator}
 rotCoat_2 {type=Rotator}
 rotCoat_3 {type=Rotator}
 rotCoat_4 {type=Rotator}
 rotCoat_5 {type=Rotator}
 rotCoat_6 {type=Rotator}
 rotCoat_7 {type=Rotator}
 rotCoat_foot_0 {type=Rotator}
 rotCoat_foot_1 {type=Rotator}
 rotCoat_foot_2 {type=Rotator}
 rotCoat_foot_3 {type=Rotator}
 scaleCoat {type=Vector}
 scaleCoatAlpha {type=number}
 scaleCoatMid {type=Vector}
 scaleCoatRight {type=Vector}
 simSpaceRoot {type=number}
 simSpaceSpine {type=number}
 thigh_r_crouch_pitch_alpha {type=number}
 thigh_r_crouch_yaw_alpha {type=number}
 windVector {type=Vector}
]
